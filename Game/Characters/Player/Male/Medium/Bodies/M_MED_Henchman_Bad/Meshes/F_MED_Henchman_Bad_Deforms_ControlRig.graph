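# Control Rig for the Henchman Bad player body that drives corrective deform bones at shoulders, wrists, hips, knees, elbows, biceps and lats.
# Parent constraints blend most deform base bones between neighboring skeleton bones.
# Spherical pose readers, or clamped yaw deltas from reference pose for knees and elbows, weight offsets from the deform_value variable.
# Right-side offsets are mirrored from left values, and boolean inputs gate most regions.
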ControlRigBlueprint /Game/Characters/Player/Male/Medium/Bodies/M_MED_Henchman_Bad/Meshes/F_MED_Henchman_Bad_Deforms_ControlRig
graphs [
 ContainedGraph {nodes [
   n1 = RigVMFunctionEntryNode()
   SequenceAggregate_1 = &RigVMFunction_Sequence(ExecuteContext=@n1.ExecuteContext)
   SequenceAggregate_1_1 = &RigVMFunction_Sequence(ExecuteContext=@SequenceAggregate_1.B)
   SequenceAggregate_1_2 = &RigVMFunction_Sequence(ExecuteContext=@SequenceAggregate_1_1.B)
   RigVMFunctionReturnNode(A=@SequenceAggregate_1.A, B=@SequenceAggregate_1_1.A, C=@SequenceAggregate_1_2.A, D=@SequenceAggregate_1_2.B)
  ]}
 ContainedGraph {nodes [
   n1 = RigVMFunctionEntryNode()
   SequenceAggregate_2 = &RigVMFunction_Sequence(ExecuteContext=@n1.ExecuteContext)
   SequenceAggregate_2_1 = &RigVMFunction_Sequence(ExecuteContext=@SequenceAggregate_2.B)
   RigVMFunctionReturnNode(A=@SequenceAggregate_2.A, B=@SequenceAggregate_2_1.A, C=@SequenceAggregate_2_1.B)
  ]}
 ContainedGraph {nodes [
   n1 = RigVMFunctionEntryNode()
   n2 = &RigUnit_GetTransform(Space=LocalSpace, Item=@n1.Item)
   n3 = &RigUnit_GetTransform(Space=LocalSpace, bInitial=true, Item=@n1.Item)
   > RigVMFunctionReturnNode(OutTransform=/Script/RigVM.RigVMFunction_MathTransformToEulerTransform(Value=&RigVMFunction_MathTransformMakeRelative(Global=@n2.Transform, Parent=@n3.Transform).Local) {is=struct RigVMUnitNode, template="Cast::Execute(in Value,out Result)"}.Result)
  ]}
 ContainedGraph {nodes [
   n1 = RigVMFunctionEntryNode()
   > SequenceExecution_2 = RigVMAggregateNode()
   n2 = &RigVMFunction_ControlFlowBranch(Condition=@n1.upperarm_up, ExecuteContext=@SequenceExecution_2.A)
   SequenceExecution_3 = RigVMAggregateNode(ExecuteContext=@n2.True)
   &RigUnit_ParentConstraint(
    AdvancedSettings=(RotationOrderForFilter=XZY)
    Child=(Type=Bone,Name="deform_upperarm_up_base_l")
    Parents=((Item=(Type=Bone,Name="upperarm_twist_01_l"),Weight=0.5),(Item=(Type=Bone,Name="clavicle_l"),Weight=0.5))
    Weight=1
    ExecutePin=@SequenceExecution_3.A
   )
   > n3 = &RigUnit_SphericalPoseReader(
    ActiveRegionScaleFactors=(PositiveWidth=2.5,NegativeWidth=3,PositiveHeight=3,NegativeHeight=3)
    ActiveRegionSize=0.1
    Debug=(bDrawDebug=false,DebugScale=25,DebugSegments=20,DebugThickness=0.25)
    DriverAxis=(X=1,Y=0,Z=0)
    DriverItem=(Type=Bone,Name="upperarm_l")
    FalloffRegionScaleFactors=(PositiveWidth=1,NegativeWidth=1,PositiveHeight=1,NegativeHeight=1)
    FalloffSize=0.5
    OptionalParentItem=(Type=Bone)
    RotationOffset=(X=0,Y=0,Z=0)
    Debug.bDrawDebug=@n1.debug_up
   )
   > &RigUnit_ModifyTransforms(
    ItemToModify=((Item=(Type=Bone,Name="deform_upperarm_up_l"),Transform=(Rotation=(X=0,Y=0,Z=0,W=1),Translation=(X=0,Y=0,Z=0),Scale3D=(X=1,Y=1,Z=1))))
    ItemToModify.0.Transform=RigVMVariableNode() {variable=deform_value}.Value.upperarm_up_18_#1
    Weight=@n3.OutputParam
   )
   &RigUnit_ParentConstraint(
    AdvancedSettings=(RotationOrderForFilter=XZY)
    Child=(Type=Bone,Name="deform_upperarm_up_base_r")
    Parents=((Item=(Type=Bone,Name="upperarm_twist_01_r"),Weight=0.5),(Item=(Type=Bone,Name="clavicle_r"),Weight=0.5))
    Weight=1
    ExecutePin=@SequenceExecution_3.B
   )
   > n4 = &RigUnit_SphericalPoseReader(
    ActiveRegionScaleFactors=(PositiveWidth=2.5,NegativeWidth=3,PositiveHeight=3,NegativeHeight=3)
    ActiveRegionSize=0.1
    Debug=(bDrawDebug=false,DebugScale=25,DebugSegments=20,DebugThickness=0.25)
    DriverAxis=(X=-1,Y=0,Z=0)
    DriverItem=(Type=Bone,Name="upperarm_r")
    FalloffRegionScaleFactors=(PositiveWidth=1,NegativeWidth=1,PositiveHeight=1,NegativeHeight=1)
    FalloffSize=0.5
    FlipHeightScaling=true
    OptionalParentItem=(Type=Bone)
    RotationOffset=(X=180,Y=0,Z=180)
    Debug.bDrawDebug=@n1.debug_up
   )
   > Deform_GetMirrorTransform_FNC = RigVMFunctionReferenceNode(transform=RigVMVariableNode() {variable=deform_value}.Value.upperarm_up_18_#1)
   > &RigUnit_ModifyTransforms(
    ItemToModify=((Item=(Type=Bone,Name="deform_upperarm_up_r"),Transform=(Rotation=(X=0,Y=0,Z=0,W=1),Translation=(X=0,Y=0,Z=0),Scale3D=(X=1,Y=1,Z=1))))
    ItemToModify.0.Transform=@Deform_GetMirrorTransform_FNC.mirrorTransform
    Weight=@n4.OutputParam
   )
   n5 = &RigVMFunction_ControlFlowBranch(Condition=@n1.upperarm_fr, ExecuteContext=@SequenceExecution_2.B)
   SequenceExecution_3_1 = RigVMAggregateNode(ExecuteContext=@n5.True)
   &RigUnit_ParentConstraint(
    AdvancedSettings=(RotationOrderForFilter=XZY)
    Child=(Type=Bone,Name="deform_upperarm_fr_base_l")
    Parents=((Item=(Type=Bone,Name="upperarm_twist_01_l"),Weight=0.3),(Item=(Type=Bone,Name="clavicle_l"),Weight=0.7))
    Weight=1
    ExecutePin=@SequenceExecution_3_1.A
   )
   > n6 = &RigUnit_SphericalPoseReader(
    ActiveRegionScaleFactors=(PositiveWidth=0.5,NegativeWidth=3,PositiveHeight=1.5,NegativeHeight=1.5)
    ActiveRegionSize=0.1
    Debug=(bDrawDebug=false,DebugScale=25,DebugSegments=20,DebugThickness=0.25)
    DriverAxis=(X=1,Y=0,Z=0)
    DriverItem=(Type=Bone,Name="upperarm_l")
    FalloffRegionScaleFactors=(PositiveWidth=1,NegativeWidth=1,PositiveHeight=1,NegativeHeight=1)
    FalloffSize=0.4
    OptionalParentItem=(Type=Bone)
    RotationOffset=(X=90,Y=0,Z=0)
    Debug.bDrawDebug=@n1.debug_fr
   )
   n7 = RigVMVariableNode() {variable=deform_value}
   > &RigUnit_ModifyTransforms(
    ItemToModify=((Item=(Type=Bone,Name="deform_upperarm_fr_l"),Transform=(Rotation=(X=0,Y=0,Z=0,W=1),Translation=(X=0,Y=0,Z=0),Scale3D=(X=1,Y=1,Z=1))))
    ItemToModify.0.Transform=@n7.Value.upperarm_fwd_fr_19_#2
    Weight=@n6.OutputParam
   )
   n8 = RigVMVariableNode() {variable=deform_value}
   > &RigUnit_ModifyTransforms(
    ItemToModify=((Item=(Type=Bone,Name="deform_upperarm_bk_l"),Transform=(Rotation=(X=0,Y=0,Z=0,W=1),Translation=(X=0,Y=0,Z=0),Scale3D=(X=1,Y=1,Z=1))))
    ItemToModify.0.Transform=@n8.Value.upperarm_fwd_bk_20_#3
    Weight=@n6.OutputParam
   )
   &RigUnit_ParentConstraint(
    AdvancedSettings=(RotationOrderForFilter=XZY)
    Child=(Type=Bone,Name="deform_upperarm_fr_base_r")
    Parents=((Item=(Type=Bone,Name="upperarm_twist_01_r"),Weight=0.3),(Item=(Type=Bone,Name="clavicle_r"),Weight=0.7))
    Weight=1
    ExecutePin=@SequenceExecution_3_1.B
   )
   > n9 = &RigUnit_SphericalPoseReader(
    ActiveRegionScaleFactors=(PositiveWidth=0.5,NegativeWidth=3,PositiveHeight=1.5,NegativeHeight=1.5)
    ActiveRegionSize=0.1
    Debug=(bDrawDebug=false,DebugScale=25,DebugSegments=20,DebugThickness=0.25)
    DriverAxis=(X=-1,Y=0,Z=0)
    DriverItem=(Type=Bone,Name="upperarm_r")
    FalloffRegionScaleFactors=(PositiveWidth=1,NegativeWidth=1,PositiveHeight=1,NegativeHeight=1)
    FalloffSize=0.4
    FlipWidthScaling=true
    OptionalParentItem=(Type=Bone)
    RotationOffset=(X=90,Y=180,Z=180)
    Debug.bDrawDebug=@n1.debug_fr
   )
   > Deform_GetMirrorTransform_FNC_2 = RigVMFunctionReferenceNode(transform=@n7.Value.upperarm_fwd_fr_19_#2)
   > &RigUnit_ModifyTransforms(
    ItemToModify=((Item=(Type=Bone,Name="deform_upperarm_fr_r"),Transform=(Rotation=(X=0,Y=0,Z=0,W=1),Translation=(X=0,Y=0,Z=0),Scale3D=(X=1,Y=1,Z=1))))
    ItemToModify.0.Transform=@Deform_GetMirrorTransform_FNC_2.mirrorTransform
    Weight=@n9.OutputParam
   )
   > Deform_GetMirrorTransform_FNC_1 = RigVMFunctionReferenceNode(transform=@n8.Value.upperarm_fwd_bk_20_#3)
   > &RigUnit_ModifyTransforms(
    ItemToModify=((Item=(Type=Bone,Name="deform_upperarm_bk_r"),Transform=(Rotation=(X=0,Y=0,Z=0,W=1),Translation=(X=0,Y=0,Z=0),Scale3D=(X=1,Y=1,Z=1))))
    ItemToModify.0.Transform=@Deform_GetMirrorTransform_FNC_1.mirrorTransform
    Weight=@n9.OutputParam
   )
   n10 = &RigVMFunction_ControlFlowBranch(Condition=@n1.upperarm_bk, ExecuteContext=@SequenceExecution_2.C)
   SequenceExecution_3_1_1 = RigVMAggregateNode(ExecuteContext=@n10.True)
   &RigUnit_ParentConstraint(
    AdvancedSettings=(RotationOrderForFilter=XZY)
    Child=(Type=Bone,Name="deform_upperarm_bk_base_l")
    Parents=((Item=(Type=Bone,Name="upperarm_twist_01_l"),Weight=0.3),(Item=(Type=Bone,Name="clavicle_l"),Weight=0.7))
    Weight=1
    ExecutePin=@SequenceExecution_3_1_1.A
   )
   > n11 = &RigUnit_SphericalPoseReader(
    ActiveRegionScaleFactors=(PositiveWidth=3,NegativeWidth=1,PositiveHeight=1.5,NegativeHeight=1.5)
    ActiveRegionSize=0.1
    Debug=(bDrawDebug=false,DebugScale=25,DebugSegments=20,DebugThickness=0.25)
    DriverAxis=(X=1,Y=0,Z=0)
    DriverItem=(Type=Bone,Name="upperarm_l")
    FalloffRegionScaleFactors=(PositiveWidth=1,NegativeWidth=1,PositiveHeight=1,NegativeHeight=1)
    FalloffSize=0.4
    OptionalParentItem=(Type=Bone)
    RotationOffset=(X=90,Y=0,Z=180)
    Debug.bDrawDebug=@n1.debug_bk
   )
   n12 = RigVMVariableNode() {variable=deform_value}
   > &RigUnit_ModifyTransforms(
    ItemToModify=((Item=(Type=Bone,Name="deform_upperarm_bk_l"),Transform=(Rotation=(X=0,Y=0,Z=0,W=1),Translation=(X=0,Y=0,Z=0),Scale3D=(X=1,Y=1,Z=1))))
    Weight=@n11.OutputParam
    ItemToModify.0.Transform=@n12.Value.upperarm_bk_22_#4
   )
   &RigUnit_ParentConstraint(
    AdvancedSettings=(RotationOrderForFilter=XZY)
    Child=(Type=Bone,Name="deform_upperarm_bk_base_r")
    Parents=((Item=(Type=Bone,Name="upperarm_twist_01_r"),Weight=0.3),(Item=(Type=Bone,Name="clavicle_r"),Weight=0.7))
    Weight=1
    ExecutePin=@SequenceExecution_3_1_1.B
   )
   > n13 = &RigUnit_SphericalPoseReader(
    ActiveRegionScaleFactors=(PositiveWidth=3,NegativeWidth=1,PositiveHeight=1.5,NegativeHeight=1.5)
    ActiveRegionSize=0.1
    Debug=(bDrawDebug=false,DebugScale=25,DebugSegments=20,DebugThickness=0.25)
    DriverAxis=(X=-1,Y=0,Z=0)
    DriverItem=(Type=Bone,Name="upperarm_r")
    FalloffRegionScaleFactors=(PositiveWidth=1,NegativeWidth=1,PositiveHeight=1,NegativeHeight=1)
    FalloffSize=0.4
    FlipHeightScaling=true
    FlipWidthScaling=false
    OptionalParentItem=(Type=Bone)
    RotationOffset=(X=90,Y=0,Z=0)
    Debug.bDrawDebug=@n1.debug_bk
   )
   > Deform_GetMirrorTransform_FNC_3 = RigVMFunctionReferenceNode(transform=@n12.Value.upperarm_bk_22_#4)
   > &RigUnit_ModifyTransforms(
    ItemToModify=((Item=(Type=Bone,Name="deform_upperarm_bk_r"),Transform=(Rotation=(X=0,Y=0,Z=0,W=1),Translation=(X=0,Y=0,Z=0),Scale3D=(X=1,Y=1,Z=1))))
    ItemToModify.0.Transform=@Deform_GetMirrorTransform_FNC_3.mirrorTransform
    Weight=@n13.OutputParam
   )
   RigVMFunctionReturnNode(ExecuteContext=@SequenceExecution_2.D)
  ]}
 ContainedGraph {nodes [
   n1 = RigVMFunctionEntryNode()
   SequenceExecution_2 = &RigVMFunction_Sequence(ExecuteContext=@n1.ExecuteContext)
   SequenceExecution_2_1 = &RigVMFunction_Sequence(ExecuteContext=@SequenceExecution_2.B)
   SequenceExecution_2_2 = &RigVMFunction_Sequence(ExecuteContext=@SequenceExecution_2_1.B)
   RigVMFunctionReturnNode(A=@SequenceExecution_2.A, B=@SequenceExecution_2_1.A, C=@SequenceExecution_2_2.A, D=@SequenceExecution_2_2.B)
  ]}
 ContainedGraph {nodes [
   n1 = RigVMFunctionEntryNode()
   SequenceExecution_3 = &RigVMFunction_Sequence(ExecuteContext=@n1.ExecuteContext)
   SequenceExecution_3_1 = &RigVMFunction_Sequence(ExecuteContext=@SequenceExecution_3.B)
   SequenceExecution_3_2 = &RigVMFunction_Sequence(ExecuteContext=@SequenceExecution_3_1.B)
   RigVMFunctionReturnNode(A=@SequenceExecution_3.A, B=@SequenceExecution_3_1.A, C=@SequenceExecution_3_2.A, D=@SequenceExecution_3_2.B)
  ]}
 ContainedGraph {nodes [
   n1 = RigVMFunctionEntryNode()
   SequenceExecution_3_1 = &RigVMFunction_Sequence(ExecuteContext=@n1.ExecuteContext)
   SequenceExecution_3_1_1 = &RigVMFunction_Sequence(ExecuteContext=@SequenceExecution_3_1.B)
   SequenceExecution_3_1_2 = &RigVMFunction_Sequence(ExecuteContext=@SequenceExecution_3_1_1.B)
   RigVMFunctionReturnNode(A=@SequenceExecution_3_1.A, B=@SequenceExecution_3_1_1.A, C=@SequenceExecution_3_1_2.A, D=@SequenceExecution_3_1_2.B)
  ]}
 ContainedGraph {nodes [
   n1 = RigVMFunctionEntryNode()
   SequenceExecution_3_1_1 = &RigVMFunction_Sequence(ExecuteContext=@n1.ExecuteContext)
   SequenceExecution_3_1_1_1 = &RigVMFunction_Sequence(ExecuteContext=@SequenceExecution_3_1_1.B)
   SequenceExecution_3_1_1_2 = &RigVMFunction_Sequence(ExecuteContext=@SequenceExecution_3_1_1_1.B)
   RigVMFunctionReturnNode(A=@SequenceExecution_3_1_1.A, B=@SequenceExecution_3_1_1_1.A, C=@SequenceExecution_3_1_1_2.A, D=@SequenceExecution_3_1_1_2.B)
  ]}
 ContainedGraph {nodes [
   n1 = RigVMFunctionEntryNode()
   > SequenceExecution = RigVMAggregateNode()
   n2 = &RigVMFunction_ControlFlowBranch(Condition=@n1.wrist_up, ExecuteContext=@SequenceExecution.A)
   SequenceExecution_1 = RigVMAggregateNode(ExecuteContext=@n2.True)
   &RigUnit_ParentConstraint(
    AdvancedSettings=(RotationOrderForFilter=XZY)
    Child=(Type=Bone,Name="deform_wrist_up_base_l")
    Parents=((Item=(Type=Bone,Name="hand_l"),Weight=0.4),(Item=(Type=Bone,Name="lowerarm_twist_01_l"),Weight=0.6))
    Weight=1
    ExecutePin=@SequenceExecution_1.A
   )
   > n3 = &RigUnit_SphericalPoseReader(
    ActiveRegionScaleFactors=(PositiveWidth=0.5,NegativeWidth=3.5,PositiveHeight=3,NegativeHeight=3)
    ActiveRegionSize=0.1
    Debug=(bDrawDebug=false,DebugScale=25,DebugSegments=20,DebugThickness=0.25)
    DriverAxis=(X=1,Y=0,Z=0)
    DriverItem=(Type=Bone,Name="hand_l")
    FalloffRegionScaleFactors=(PositiveWidth=1,NegativeWidth=1,PositiveHeight=1,NegativeHeight=1)
    FalloffSize=0.45
    OptionalParentItem=(Type=Bone,Name="lowerarm_twist_01_l")
    RotationOffset=(X=90,Y=0,Z=0)
    Debug.bDrawDebug=@n1.debug_up
   )
   n4 = RigVMVariableNode() {variable=deform_value}
   > &RigUnit_ModifyTransforms(
    ItemToModify=((Item=(Type=Bone,Name="deform_wrist_up_l"),Transform=(Rotation=(X=0,Y=0,Z=0,W=1),Translation=(X=0,Y=0,Z=0),Scale3D=(X=1,Y=1,Z=1))))
    ItemToModify.0.Transform=@n4.Value.wrist_up_31_#5
    Weight=@n3.OutputParam
   )
   &RigUnit_ParentConstraint(
    AdvancedSettings=(RotationOrderForFilter=XZY)
    Child=(Type=Bone,Name="deform_wrist_up_base_r")
    Parents=((Item=(Type=Bone,Name="hand_r"),Weight=0.4),(Item=(Type=Bone,Name="lowerarm_twist_01_r"),Weight=0.6))
    Weight=1
    ExecutePin=@SequenceExecution_1.B
   )
   > n5 = &RigUnit_SphericalPoseReader(
    ActiveRegionScaleFactors=(PositiveWidth=0.5,NegativeWidth=3.5,PositiveHeight=3,NegativeHeight=3)
    ActiveRegionSize=0.1
    Debug=(bDrawDebug=false,DebugScale=25,DebugSegments=20,DebugThickness=0.25)
    DriverAxis=(X=-1,Y=0,Z=0)
    DriverItem=(Type=Bone,Name="hand_r")
    FalloffRegionScaleFactors=(PositiveWidth=1,NegativeWidth=1,PositiveHeight=1,NegativeHeight=1)
    FalloffSize=0.45
    FlipHeightScaling=true
    OptionalParentItem=(Type=Bone,Name="lowerarm_twist_01_r")
    RotationOffset=(X=90,Y=0,Z=180)
    Debug.bDrawDebug=@n1.debug_up
   )
   > Deform_GetMirrorTransform_FNC = RigVMFunctionReferenceNode(transform=@n4.Value.wrist_up_31_#5)
   > &RigUnit_ModifyTransforms(
    ItemToModify=((Item=(Type=Bone,Name="deform_wrist_up_r"),Transform=(Rotation=(X=0,Y=0,Z=0,W=1),Translation=(X=0,Y=0,Z=0),Scale3D=(X=1,Y=1,Z=1))))
    ItemToModify.0.Transform=@Deform_GetMirrorTransform_FNC.mirrorTransform
    Weight=@n5.OutputParam
   )
   n6 = &RigVMFunction_ControlFlowBranch(Condition=@n1.wrist_dn, ExecuteContext=@SequenceExecution.B)
   SequenceExecution_1_1 = RigVMAggregateNode(ExecuteContext=@n6.True)
   &RigUnit_ParentConstraint(
    AdvancedSettings=(RotationOrderForFilter=XZY)
    Child=(Type=Bone,Name="deform_wrist_dn_base_l")
    Parents=((Item=(Type=Bone,Name="hand_l"),Weight=0.4),(Item=(Type=Bone,Name="lowerarm_twist_01_l"),Weight=0.6))
    Weight=1
    ExecutePin=@SequenceExecution_1_1.A
   )
   > n7 = &RigUnit_SphericalPoseReader(
    ActiveRegionScaleFactors=(PositiveWidth=1,NegativeWidth=4,PositiveHeight=3,NegativeHeight=3)
    ActiveRegionSize=0.1
    Debug=(bDrawDebug=false,DebugScale=25,DebugSegments=20,DebugThickness=0.25)
    DriverAxis=(X=1,Y=0,Z=0)
    DriverItem=(Type=Bone,Name="hand_l")
    FalloffRegionScaleFactors=(PositiveWidth=1,NegativeWidth=1,PositiveHeight=1,NegativeHeight=1)
    FalloffSize=0.4
    OptionalParentItem=(Type=Bone,Name="lowerarm_twist_01_l")
    RotationOffset=(X=-90,Y=0,Z=0)
    Debug.bDrawDebug=@n1.debug_dn
   )
   n8 = RigVMVariableNode() {variable=deform_value}
   > &RigUnit_ModifyTransforms(
    ItemToModify=((Item=(Type=Bone,Name="deform_wrist_dn_l"),Transform=(Rotation=(X=0,Y=0,Z=0,W=1),Translation=(X=0,Y=0,Z=0),Scale3D=(X=1,Y=1,Z=1))))
    ItemToModify.0.Transform=@n8.Value.wrist_dn_32_#6
    Weight=@n7.OutputParam
   )
   &RigUnit_ParentConstraint(
    AdvancedSettings=(RotationOrderForFilter=XZY)
    Child=(Type=Bone,Name="deform_wrist_dn_base_r")
    Parents=((Item=(Type=Bone,Name="hand_r"),Weight=0.4),(Item=(Type=Bone,Name="lowerarm_twist_01_r"),Weight=0.6))
    Weight=1
    ExecutePin=@SequenceExecution_1_1.B
   )
   > n9 = &RigUnit_SphericalPoseReader(
    ActiveRegionScaleFactors=(PositiveWidth=1,NegativeWidth=4,PositiveHeight=3,NegativeHeight=3)
    ActiveRegionSize=0.1
    Debug=(bDrawDebug=false,DebugScale=25,DebugSegments=20,DebugThickness=0.25)
    DriverAxis=(X=-1,Y=0,Z=0)
    DriverItem=(Type=Bone,Name="hand_r")
    FalloffRegionScaleFactors=(PositiveWidth=1,NegativeWidth=1,PositiveHeight=1,NegativeHeight=1)
    FalloffSize=0.4
    FlipHeightScaling=true
    OptionalParentItem=(Type=Bone,Name="lowerarm_twist_01_r")
    RotationOffset=(X=-90,Y=0,Z=190)
    Debug.bDrawDebug=@n1.debug_dn
   )
   > Deform_GetMirrorTransform_FNC_1 = RigVMFunctionReferenceNode(transform=@n8.Value.wrist_dn_32_#6)
   > &RigUnit_ModifyTransforms(
    ItemToModify=((Item=(Type=Bone,Name="deform_wrist_dn_r"),Transform=(Rotation=(X=0,Y=0,Z=0,W=1),Translation=(X=0,Y=0,Z=0),Scale3D=(X=1,Y=1,Z=1))))
    ItemToModify.0.Transform=@Deform_GetMirrorTransform_FNC_1.mirrorTransform
    Weight=@n9.OutputParam
   )
   RigVMFunctionReturnNode(ExecuteContext=@SequenceExecution.D)
  ]}
 ContainedGraph {nodes [
   n1 = RigVMFunctionEntryNode()
   SequenceExecution = &RigVMFunction_Sequence(ExecuteContext=@n1.ExecuteContext)
   SequenceExecution_1 = &RigVMFunction_Sequence(ExecuteContext=@SequenceExecution.B)
   SequenceExecution_2 = &RigVMFunction_Sequence(ExecuteContext=@SequenceExecution_1.B)
   RigVMFunctionReturnNode(A=@SequenceExecution.A, B=@SequenceExecution_1.A, C=@SequenceExecution_2.A, D=@SequenceExecution_2.B)
  ]}
 ContainedGraph {nodes [
   n1 = RigVMFunctionEntryNode()
   SequenceExecution_1 = &RigVMFunction_Sequence(ExecuteContext=@n1.ExecuteContext)
   SequenceExecution_1_1 = &RigVMFunction_Sequence(ExecuteContext=@SequenceExecution_1.B)
   SequenceExecution_1_2 = &RigVMFunction_Sequence(ExecuteContext=@SequenceExecution_1_1.B)
   RigVMFunctionReturnNode(A=@SequenceExecution_1.A, B=@SequenceExecution_1_1.A, C=@SequenceExecution_1_2.A, D=@SequenceExecution_1_2.B)
  ]}
 ContainedGraph {nodes [
   n1 = RigVMFunctionEntryNode()
   SequenceExecution_1_1 = &RigVMFunction_Sequence(ExecuteContext=@n1.ExecuteContext)
   SequenceExecution_1_1_1 = &RigVMFunction_Sequence(ExecuteContext=@SequenceExecution_1_1.B)
   SequenceExecution_1_1_2 = &RigVMFunction_Sequence(ExecuteContext=@SequenceExecution_1_1_1.B)
   RigVMFunctionReturnNode(A=@SequenceExecution_1_1.A, B=@SequenceExecution_1_1_1.A, C=@SequenceExecution_1_1_2.A, D=@SequenceExecution_1_1_2.B)
  ]}
 ContainedGraph {nodes [
   n1 = RigVMFunctionEntryNode()
   n2 = &RigVMFunction_ControlFlowBranch(ExecuteContext=@n1.ExecuteContext, Condition=&RigVMFunction_MathBoolOr(B=@n1.glute, A=@n1.groin).Result)
   RigVMFunctionReturnNode(ExecuteContext=@n2.Completed)
   SequenceExecution = RigVMAggregateNode(ExecuteContext=@n2.True)
   n3 = &RigUnit_SphericalPoseReader(
    ActiveRegionScaleFactors=(PositiveWidth=3.5,NegativeWidth=0.5,PositiveHeight=3,NegativeHeight=3)
    ActiveRegionSize=0.1
    Debug=(bDrawDebug=false,DebugScale=25,DebugSegments=20,DebugThickness=0.25)
    DriverAxis=(X=-1,Y=0,Z=0)
    DriverItem=(Type=Bone,Name="thigh_l")
    FalloffRegionScaleFactors=(PositiveWidth=1,NegativeWidth=1,PositiveHeight=1,NegativeHeight=1)
    FalloffSize=0.45
    OptionalParentItem=(Type=Bone)
    RotationOffset=(X=90,Y=0,Z=0)
    ExecutePin=@SequenceExecution.A
    Debug.bDrawDebug=@n1.debug
   )
   > n4 = &RigUnit_SphericalPoseReader(
    ActiveRegionScaleFactors=(PositiveWidth=3.5,NegativeWidth=0.5,PositiveHeight=3,NegativeHeight=3)
    ActiveRegionSize=0.1
    Debug=(bDrawDebug=false,DebugScale=25,DebugSegments=20,DebugThickness=0.25)
    DriverAxis=(X=1,Y=0,Z=0)
    DriverItem=(Type=Bone,Name="thigh_r")
    FalloffRegionScaleFactors=(PositiveWidth=1,NegativeWidth=1,PositiveHeight=1,NegativeHeight=1)
    FalloffSize=0.45
    FlipHeightScaling=false
    FlipWidthScaling=true
    OptionalParentItem=(Type=Bone)
    RotationOffset=(X=-90,Y=0,Z=0)
    Debug.bDrawDebug=@n1.debug
   )
   n5 = &RigVMFunction_ControlFlowBranch(ExecuteContext=@SequenceExecution.B, Condition=@n1.groin)
   SequenceExecution_1 = RigVMAggregateNode(ExecuteContext=@n5.True)
   &RigUnit_ParentConstraint(
    AdvancedSettings=(RotationOrderForFilter=XZY)
    Child=(Type=Bone,Name="deform_groin_base_l")
    Parents=((Item=(Type=Bone,Name="thigh_l"),Weight=0.7),(Item=(Type=Bone,Name="pelvis"),Weight=0.3))
    Weight=1
    ExecutePin=@SequenceExecution_1.A
   )
   n6 = RigVMVariableNode() {variable=deform_value}
   > &RigUnit_ModifyTransforms(
    ItemToModify=((Item=(Type=Bone,Name="deform_groin_l"),Transform=(Rotation=(X=0,Y=0,Z=0,W=1),Translation=(X=0,Y=0,Z=0),Scale3D=(X=1,Y=1,Z=1))))
    ItemToModify.0.Transform=@n6.Value.groin_35_#7
    Weight=@n3.OutputParam
   )
   &RigUnit_ParentConstraint(
    AdvancedSettings=(RotationOrderForFilter=XZY)
    Child=(Type=Bone,Name="deform_groin_base_r")
    Parents=((Item=(Type=Bone,Name="thigh_r"),Weight=0.7),(Item=(Type=Bone,Name="pelvis"),Weight=0.3))
    Weight=1
    ExecutePin=@SequenceExecution_1.B
   )
   > Deform_GetMirrorTransform_FNC = RigVMFunctionReferenceNode(transform=@n6.Value.groin_35_#7)
   > &RigUnit_ModifyTransforms(
    ItemToModify=((Item=(Type=Bone,Name="deform_groin_r"),Transform=(Rotation=(X=0,Y=0,Z=0,W=1),Translation=(X=0,Y=0,Z=0),Scale3D=(X=1,Y=1,Z=1))))
    ItemToModify.0.Transform=@Deform_GetMirrorTransform_FNC.mirrorTransform
    Weight=@n4.OutputParam
   )
   n7 = &RigVMFunction_ControlFlowBranch(ExecuteContext=@SequenceExecution.C, Condition=@n1.glute)
   SequenceExecution_1_1 = RigVMAggregateNode(ExecuteContext=@n7.True)
   &RigUnit_ParentConstraint(
    AdvancedSettings=(RotationOrderForFilter=XZY)
    Child=(Type=Bone,Name="deform_glute_base_l")
    Parents=((Item=(Type=Bone,Name="thigh_l"),Weight=0.25),(Item=(Type=Bone,Name="pelvis"),Weight=0.75))
    Weight=1
    ExecutePin=@SequenceExecution_1_1.A
   )
   n8 = RigVMVariableNode() {variable=deform_value}
   > &RigUnit_ModifyTransforms(
    ItemToModify=((Item=(Type=Bone,Name="deform_glute_l"),Transform=(Rotation=(X=0,Y=0,Z=0,W=1),Translation=(X=0,Y=0,Z=0),Scale3D=(X=1,Y=1,Z=1))))
    ItemToModify.0.Transform=@n8.Value.glute_36_#8
    Weight=@n3.OutputParam
   )
   &RigUnit_ParentConstraint(
    AdvancedSettings=(RotationOrderForFilter=XZY)
    Child=(Type=Bone,Name="deform_glute_base_r")
    Parents=((Item=(Type=Bone,Name="thigh_r"),Weight=0.25),(Item=(Type=Bone,Name="pelvis"),Weight=0.75))
    Weight=1
    ExecutePin=@SequenceExecution_1_1.B
   )
   > Deform_GetMirrorTransform_FNC_1 = RigVMFunctionReferenceNode(transform=@n8.Value.glute_36_#8)
   > &RigUnit_ModifyTransforms(
    ItemToModify=((Item=(Type=Bone,Name="deform_glute_r"),Transform=(Rotation=(X=0,Y=0,Z=0,W=1),Translation=(X=0,Y=0,Z=0),Scale3D=(X=1,Y=1,Z=1))))
    ItemToModify.0.Transform=@Deform_GetMirrorTransform_FNC_1.mirrorTransform
    Weight=@n4.OutputParam
   )
  ]}
 ContainedGraph {nodes [
   n1 = RigVMFunctionEntryNode()
   SequenceExecution = &RigVMFunction_Sequence(ExecuteContext=@n1.ExecuteContext)
   SequenceExecution_1 = &RigVMFunction_Sequence(ExecuteContext=@SequenceExecution.B)
   SequenceExecution_2 = &RigVMFunction_Sequence(ExecuteContext=@SequenceExecution_1.B)
   RigVMFunctionReturnNode(A=@SequenceExecution.A, B=@SequenceExecution_1.A, C=@SequenceExecution_2.A, D=@SequenceExecution_2.B)
  ]}
 ContainedGraph {nodes [
   n1 = RigVMFunctionEntryNode()
   SequenceExecution_1 = &RigVMFunction_Sequence(ExecuteContext=@n1.ExecuteContext)
   SequenceExecution_1_1 = &RigVMFunction_Sequence(ExecuteContext=@SequenceExecution_1.B)
   SequenceExecution_1_2 = &RigVMFunction_Sequence(ExecuteContext=@SequenceExecution_1_1.B)
   RigVMFunctionReturnNode(A=@SequenceExecution_1.A, B=@SequenceExecution_1_1.A, C=@SequenceExecution_1_2.A, D=@SequenceExecution_1_2.B)
  ]}
 ContainedGraph {nodes [
   n1 = RigVMFunctionEntryNode()
   SequenceExecution_1_1 = &RigVMFunction_Sequence(ExecuteContext=@n1.ExecuteContext)
   SequenceExecution_1_1_1 = &RigVMFunction_Sequence(ExecuteContext=@SequenceExecution_1_1.B)
   SequenceExecution_1_1_2 = &RigVMFunction_Sequence(ExecuteContext=@SequenceExecution_1_1_1.B)
   RigVMFunctionReturnNode(A=@SequenceExecution_1_1.A, B=@SequenceExecution_1_1_1.A, C=@SequenceExecution_1_1_2.A, D=@SequenceExecution_1_1_2.B)
  ]}
 ContainedGraph {nodes [
   n1 = RigVMFunctionEntryNode()
   > SequenceExecution_2 = RigVMAggregateNode()
   GetDeltaTransformFromRefPose_FNC = RigVMFunctionReferenceNode(Item=(Type=Bone,Name="calf_l"), ExecuteContext=@SequenceExecution_2.A)
   GetDeltaTransformFromRefPose_FNC_1 = RigVMFunctionReferenceNode(Item=(Type=Bone,Name="calf_r"), ExecuteContext=@SequenceExecution_2.B)
   SequenceExecution = RigVMAggregateNode(ExecuteContext=@SequenceExecution_2.C)
   n2 = &RigVMFunction_ControlFlowBranch(Condition=@n1.knee_in, ExecuteContext=@SequenceExecution.A)
   SequenceExecution_1 = RigVMAggregateNode(ExecuteContext=@n2.True)
   &RigUnit_ParentConstraint(
    AdvancedSettings=(RotationOrderForFilter=XZY)
    Child=(Type=Bone,Name="deform_knee_in_base_l")
    Parents=((Item=(Type=Bone,Name="calf_l"),Weight=0.5),(Item=(Type=Bone,Name="thigh_l"),Weight=0.5))
    Weight=1
    ExecutePin=@SequenceExecution_1.A
   )
   n3 = RigVMVariableNode() {variable=deform_value}
   n4 = &RigVMFunction_MathFloatRemap(SourceMaximum=100, SourceMinimum=10, bClamp=true, Value=@GetDeltaTransformFromRefPose_FNC.OutTransform.Rotation.Yaw)
   > &RigUnit_ModifyTransforms(
    ItemToModify=((Item=(Type=Bone,Name="deform_knee_in_l"),Transform=(Rotation=(X=0,Y=0,Z=0,W=1),Translation=(X=0,Y=0,Z=0),Scale3D=(X=1,Y=1,Z=1))))
    ItemToModify.0.Transform=@n3.Value.knee_in_41_#9
    Weight=@n4.Result
   )
   &RigUnit_ParentConstraint(
    AdvancedSettings=(RotationOrderForFilter=XZY)
    Child=(Type=Bone,Name="deform_knee_in_base_r")
    Parents=((Item=(Type=Bone,Name="calf_r"),Weight=0.5),(Item=(Type=Bone,Name="thigh_r"),Weight=0.5))
    Weight=1
    ExecutePin=@SequenceExecution_1.B
   )
   > Deform_GetMirrorTransform_FNC = RigVMFunctionReferenceNode(transform=@n3.Value.knee_in_41_#9)
   n5 = &RigVMFunction_MathFloatRemap(SourceMaximum=100, SourceMinimum=10, bClamp=true, Value=@GetDeltaTransformFromRefPose_FNC_1.OutTransform.Rotation.Yaw)
   > &RigUnit_ModifyTransforms(
    ItemToModify=((Item=(Type=Bone,Name="deform_knee_in_r"),Transform=(Rotation=(X=0,Y=0,Z=0,W=1),Translation=(X=0,Y=0,Z=0),Scale3D=(X=1,Y=1,Z=1))))
    ItemToModify.0.Transform=@Deform_GetMirrorTransform_FNC.mirrorTransform
    Weight=@n5.Result
   )
   n6 = &RigVMFunction_ControlFlowBranch(Condition=@n1.knee_out, ExecuteContext=@SequenceExecution.B)
   SequenceExecution_1_1 = RigVMAggregateNode(ExecuteContext=@n6.True)
   &RigUnit_ParentConstraint(
    AdvancedSettings=(RotationOrderForFilter=XZY)
    Child=(Type=Bone,Name="deform_knee_out_base_l")
    Parents=((Item=(Type=Bone,Name="calf_l"),Weight=0.1),(Item=(Type=Bone,Name="thigh_l"),Weight=0.9))
    Weight=1
    ExecutePin=@SequenceExecution_1_1.A
   )
   n7 = RigVMVariableNode() {variable=deform_value}
   > &RigUnit_ModifyTransforms(
    ItemToModify=((Item=(Type=Bone,Name="deform_knee_out_l"),Transform=(Rotation=(X=0,Y=0,Z=0,W=1),Translation=(X=0,Y=0,Z=0),Scale3D=(X=1,Y=1,Z=1))))
    ItemToModify.0.Transform=@n7.Value.knee_out_42_#10
    Weight=@n4.Result
   )
   &RigUnit_ParentConstraint(
    AdvancedSettings=(RotationOrderForFilter=XZY)
    Child=(Type=Bone,Name="deform_knee_out_base_r")
    Parents=((Item=(Type=Bone,Name="calf_r"),Weight=0.1),(Item=(Type=Bone,Name="thigh_r"),Weight=0.9))
    Weight=1
    ExecutePin=@SequenceExecution_1_1.B
   )
   > Deform_GetMirrorTransform_FNC_1 = RigVMFunctionReferenceNode(transform=@n7.Value.knee_out_42_#10)
   > &RigUnit_ModifyTransforms(
    ItemToModify=((Item=(Type=Bone,Name="deform_knee_out_r"),Transform=(Rotation=(X=0,Y=0,Z=0,W=1),Translation=(X=0,Y=0,Z=0),Scale3D=(X=1,Y=1,Z=1))))
    ItemToModify.0.Transform=@Deform_GetMirrorTransform_FNC_1.mirrorTransform
    Weight=@n5.Result
   )
   RigVMFunctionReturnNode(ExecuteContext=@SequenceExecution_2.D)
  ]}
 ContainedGraph {nodes [
   n1 = RigVMFunctionEntryNode()
   SequenceExecution_2 = &RigVMFunction_Sequence(ExecuteContext=@n1.ExecuteContext)
   SequenceExecution_2_1 = &RigVMFunction_Sequence(ExecuteContext=@SequenceExecution_2.B)
   SequenceExecution_2_2 = &RigVMFunction_Sequence(ExecuteContext=@SequenceExecution_2_1.B)
   RigVMFunctionReturnNode(A=@SequenceExecution_2.A, B=@SequenceExecution_2_1.A, C=@SequenceExecution_2_2.A, D=@SequenceExecution_2_2.B)
  ]}
 ContainedGraph {nodes [
   n1 = RigVMFunctionEntryNode()
   SequenceExecution = &RigVMFunction_Sequence(ExecuteContext=@n1.ExecuteContext)
   SequenceExecution_1 = &RigVMFunction_Sequence(ExecuteContext=@SequenceExecution.B)
   SequenceExecution_2 = &RigVMFunction_Sequence(ExecuteContext=@SequenceExecution_1.B)
   RigVMFunctionReturnNode(A=@SequenceExecution.A, B=@SequenceExecution_1.A, C=@SequenceExecution_2.A, D=@SequenceExecution_2.B)
  ]}
 ContainedGraph {nodes [
   n1 = RigVMFunctionEntryNode()
   SequenceExecution_1 = &RigVMFunction_Sequence(ExecuteContext=@n1.ExecuteContext)
   SequenceExecution_1_1 = &RigVMFunction_Sequence(ExecuteContext=@SequenceExecution_1.B)
   SequenceExecution_1_2 = &RigVMFunction_Sequence(ExecuteContext=@SequenceExecution_1_1.B)
   RigVMFunctionReturnNode(A=@SequenceExecution_1.A, B=@SequenceExecution_1_1.A, C=@SequenceExecution_1_2.A, D=@SequenceExecution_1_2.B)
  ]}
 ContainedGraph {nodes [
   n1 = RigVMFunctionEntryNode()
   SequenceExecution_1_1 = &RigVMFunction_Sequence(ExecuteContext=@n1.ExecuteContext)
   SequenceExecution_1_1_1 = &RigVMFunction_Sequence(ExecuteContext=@SequenceExecution_1_1.B)
   SequenceExecution_1_1_2 = &RigVMFunction_Sequence(ExecuteContext=@SequenceExecution_1_1_1.B)
   RigVMFunctionReturnNode(A=@SequenceExecution_1_1.A, B=@SequenceExecution_1_1_1.A, C=@SequenceExecution_1_1_2.A, D=@SequenceExecution_1_1_2.B)
  ]}
 ContainedGraph {nodes [
   n1 = RigVMFunctionEntryNode()
   > SequenceExecution_2 = RigVMAggregateNode()
   GetDeltaTransformFromRefPose_FNC = RigVMFunctionReferenceNode(Item=(Type=Bone,Name="lowerarm_l"), ExecuteContext=@SequenceExecution_2.A)
   GetDeltaTransformFromRefPose_FNC_1 = RigVMFunctionReferenceNode(Item=(Type=Bone,Name="lowerarm_r"), ExecuteContext=@SequenceExecution_2.B)
   SequenceExecution = RigVMAggregateNode(ExecuteContext=@SequenceExecution_2.C)
   n2 = &RigVMFunction_ControlFlowBranch(ExecuteContext=@SequenceExecution.A, Condition=@n1.elbow_in)
   SequenceExecution_1 = RigVMAggregateNode(ExecuteContext=@n2.True)
   &RigUnit_ParentConstraint(
    AdvancedSettings=(RotationOrderForFilter=XZY)
    Child=(Type=Bone,Name="deform_elbow_in_base_l")
    Parents=((Item=(Type=Bone,Name="lowerarm_l"),Weight=0.4),(Item=(Type=Bone,Name="upperarm_l"),Weight=0.6))
    Weight=1
    ExecutePin=@SequenceExecution_1.A
   )
   n3 = RigVMVariableNode() {variable=deform_value}
   n4 = &RigVMFunction_MathFloatRemap(SourceMaximum=90, SourceMinimum=15, bClamp=true, Value=@GetDeltaTransformFromRefPose_FNC.OutTransform.Rotation.Yaw)
   > &RigUnit_ModifyTransforms(
    ItemToModify=((Item=(Type=Bone,Name="deform_elbow_in_l"),Transform=(Rotation=(X=0,Y=0,Z=0,W=1),Translation=(X=0,Y=0,Z=0),Scale3D=(X=1,Y=1,Z=1))))
    ItemToModify.0.Transform=@n3.Value.elbow_in_24_#11
    Weight=@n4.Result
   )
   &RigUnit_ParentConstraint(
    AdvancedSettings=(RotationOrderForFilter=XZY)
    Child=(Type=Bone,Name="deform_elbow_in_base_r")
    Parents=((Item=(Type=Bone,Name="lowerarm_r"),Weight=0.4),(Item=(Type=Bone,Name="upperarm_r"),Weight=0.6))
    Weight=1
    ExecutePin=@SequenceExecution_1.B
   )
   > Deform_GetMirrorTransform_FNC_1 = RigVMFunctionReferenceNode(transform=@n3.Value.elbow_in_24_#11)
   n5 = &RigVMFunction_MathFloatRemap(SourceMaximum=90, SourceMinimum=15, bClamp=true, Value=@GetDeltaTransformFromRefPose_FNC_1.OutTransform.Rotation.Yaw)
   > &RigUnit_ModifyTransforms(
    ItemToModify=((Item=(Type=Bone,Name="deform_elbow_in_r"),Transform=(Rotation=(X=0,Y=0,Z=0,W=1),Translation=(X=0,Y=0,Z=0),Scale3D=(X=1,Y=1,Z=1))))
    ItemToModify.0.Transform=@Deform_GetMirrorTransform_FNC_1.mirrorTransform
    Weight=@n5.Result
   )
   n6 = &RigVMFunction_ControlFlowBranch(ExecuteContext=@SequenceExecution.C, Condition=@n1.bicep)
   SequenceExecution_1_1_1 = RigVMAggregateNode(ExecuteContext=@n6.True)
   n7 = RigVMVariableNode() {variable=deform_value}
   &RigUnit_ModifyTransforms(
    ItemToModify=((Item=(Type=Bone,Name="deform_bicep_l"),Transform=(Rotation=(X=0,Y=0,Z=0,W=1),Translation=(X=0,Y=0,Z=0),Scale3D=(X=1,Y=1,Z=1))))
    ItemToModify.0.Transform=@n7.Value.bicep_44_#12
    ExecutePin=@SequenceExecution_1_1_1.A
    Weight=@n4.Result
   )
   Deform_GetMirrorTransform_FNC = RigVMFunctionReferenceNode(transform=@n7.Value.bicep_44_#12, ExecuteContext=@SequenceExecution_1_1_1.B)
   > &RigUnit_ModifyTransforms(
    ItemToModify=((Item=(Type=Bone,Name="deform_bicep_r"),Transform=(Rotation=(X=0,Y=0,Z=0,W=1),Translation=(X=0,Y=0,Z=0),Scale3D=(X=1,Y=1,Z=1))))
    ItemToModify.0.Transform=@Deform_GetMirrorTransform_FNC.mirrorTransform
    Weight=@n5.Result
   )
   n8 = &RigVMFunction_ControlFlowBranch(Condition=@n1.elbow_out, ExecuteContext=@SequenceExecution.B)
   SequenceExecution_1_1 = RigVMAggregateNode(ExecuteContext=@n8.True)
   &RigUnit_ParentConstraint(
    AdvancedSettings=(RotationOrderForFilter=XZY)
    Child=(Type=Bone,Name="deform_elbow_out_base_l")
    Parents=((Item=(Type=Bone,Name="lowerarm_l"),Weight=0.3),(Item=(Type=Bone,Name="upperarm_l"),Weight=0.7))
    Weight=1
    ExecutePin=@SequenceExecution_1_1.A
   )
   n9 = RigVMVariableNode() {variable=deform_value}
   > &RigUnit_ModifyTransforms(
    ItemToModify=((Item=(Type=Bone,Name="deform_elbow_out_l"),Transform=(Rotation=(X=0,Y=0,Z=0,W=1),Translation=(X=0,Y=0,Z=0),Scale3D=(X=1,Y=1,Z=1))))
    ItemToModify.0.Transform=@n9.Value.elbow_out_26_#13
    Weight=@n4.Result
   )
   &RigUnit_ParentConstraint(
    AdvancedSettings=(RotationOrderForFilter=XZY)
    Child=(Type=Bone,Name="deform_elbow_out_base_r")
    Parents=((Item=(Type=Bone,Name="lowerarm_r"),Weight=0.3),(Item=(Type=Bone,Name="upperarm_r"),Weight=0.7))
    Weight=1
    ExecutePin=@SequenceExecution_1_1.B
   )
   > Deform_GetMirrorTransform_FNC_2 = RigVMFunctionReferenceNode(transform=@n9.Value.elbow_out_26_#13)
   > &RigUnit_ModifyTransforms(
    ItemToModify=((Item=(Type=Bone,Name="deform_elbow_out_r"),Transform=(Rotation=(X=0,Y=0,Z=0,W=1),Translation=(X=0,Y=0,Z=0),Scale3D=(X=1,Y=1,Z=1))))
    ItemToModify.0.Transform=@Deform_GetMirrorTransform_FNC_2.mirrorTransform
    Weight=@n5.Result
   )
   RigVMFunctionReturnNode(ExecuteContext=@SequenceExecution_2.D)
  ]}
 ContainedGraph {nodes [
   n1 = RigVMFunctionEntryNode()
   SequenceExecution_2 = &RigVMFunction_Sequence(ExecuteContext=@n1.ExecuteContext)
   SequenceExecution_2_1 = &RigVMFunction_Sequence(ExecuteContext=@SequenceExecution_2.B)
   SequenceExecution_2_2 = &RigVMFunction_Sequence(ExecuteContext=@SequenceExecution_2_1.B)
   RigVMFunctionReturnNode(A=@SequenceExecution_2.A, B=@SequenceExecution_2_1.A, C=@SequenceExecution_2_2.A, D=@SequenceExecution_2_2.B)
  ]}
 ContainedGraph {nodes [
   n1 = RigVMFunctionEntryNode()
   SequenceExecution = &RigVMFunction_Sequence(ExecuteContext=@n1.ExecuteContext)
   SequenceExecution_1 = &RigVMFunction_Sequence(ExecuteContext=@SequenceExecution.B)
   SequenceExecution_2 = &RigVMFunction_Sequence(ExecuteContext=@SequenceExecution_1.B)
   RigVMFunctionReturnNode(A=@SequenceExecution.A, B=@SequenceExecution_1.A, C=@SequenceExecution_2.A, D=@SequenceExecution_2.B)
  ]}
 ContainedGraph {nodes [
   n1 = RigVMFunctionEntryNode()
   SequenceExecution_1 = &RigVMFunction_Sequence(ExecuteContext=@n1.ExecuteContext)
   SequenceExecution_1_1 = &RigVMFunction_Sequence(ExecuteContext=@SequenceExecution_1.B)
   SequenceExecution_1_2 = &RigVMFunction_Sequence(ExecuteContext=@SequenceExecution_1_1.B)
   RigVMFunctionReturnNode(A=@SequenceExecution_1.A, B=@SequenceExecution_1_1.A, C=@SequenceExecution_1_2.A, D=@SequenceExecution_1_2.B)
  ]}
 ContainedGraph {nodes [
   n1 = RigVMFunctionEntryNode()
   SequenceExecution_1_1 = &RigVMFunction_Sequence(ExecuteContext=@n1.ExecuteContext)
   SequenceExecution_1_1_1 = &RigVMFunction_Sequence(ExecuteContext=@SequenceExecution_1_1.B)
   SequenceExecution_1_1_2 = &RigVMFunction_Sequence(ExecuteContext=@SequenceExecution_1_1_1.B)
   RigVMFunctionReturnNode(A=@SequenceExecution_1_1.A, B=@SequenceExecution_1_1_1.A, C=@SequenceExecution_1_1_2.A, D=@SequenceExecution_1_1_2.B)
  ]}
 ContainedGraph {nodes [
   n1 = RigVMFunctionEntryNode()
   SequenceExecution_1_1_1 = &RigVMFunction_Sequence(ExecuteContext=@n1.ExecuteContext)
   SequenceExecution_1_1_1_1 = &RigVMFunction_Sequence(ExecuteContext=@SequenceExecution_1_1_1.B)
   SequenceExecution_1_1_1_2 = &RigVMFunction_Sequence(ExecuteContext=@SequenceExecution_1_1_1_1.B)
   RigVMFunctionReturnNode(A=@SequenceExecution_1_1_1.A, B=@SequenceExecution_1_1_1_1.A, C=@SequenceExecution_1_1_1_2.A, D=@SequenceExecution_1_1_1_2.B)
  ]}
 ContainedGraph {nodes [
   n1 = RigVMFunctionEntryNode()
   > SequenceExecution = RigVMAggregateNode()
   n2 = &RigUnit_SphericalPoseReader(
    ActiveRegionScaleFactors=(PositiveWidth=4,NegativeWidth=0.5,PositiveHeight=1.5,NegativeHeight=4)
    ActiveRegionSize=0.05
    Debug=(bDrawDebug=false,DebugScale=25,DebugSegments=20,DebugThickness=0.25)
    DriverAxis=(X=1,Y=0,Z=0)
    DriverItem=(Type=Bone,Name="upperarm_l")
    FalloffRegionScaleFactors=(PositiveWidth=1,NegativeWidth=1,PositiveHeight=1,NegativeHeight=1)
    FalloffSize=0.2
    OptionalParentItem=(Type=Bone,Name="spine_03")
    RotationOffset=(X=0,Y=-150,Z=0)
    Debug.bDrawDebug=@n1.debug
    ExecutePin=@SequenceExecution.A
   )
   n3 = RigVMVariableNode() {variable=deform_value}
   > &RigUnit_ModifyTransforms(
    ItemToModify=((Item=(Type=Bone,Name="deform_lat_l"),Transform=(Rotation=(X=0,Y=0,Z=0,W=1),Translation=(X=0,Y=0,Z=0),Scale3D=(X=1,Y=1,Z=1))))
    ItemToModify.0.Transform=@n3.Value.lat_45_#14
    Weight=@n2.OutputParam
   )
   n4 = &RigUnit_SphericalPoseReader(
    ActiveRegionScaleFactors=(PositiveWidth=4,NegativeWidth=0.5,PositiveHeight=1.5,NegativeHeight=4)
    ActiveRegionSize=0.05
    Debug=(bDrawDebug=false,DebugScale=25,DebugSegments=20,DebugThickness=0.25)
    DriverAxis=(X=-1,Y=0,Z=0)
    DriverItem=(Type=Bone,Name="upperarm_r")
    FalloffRegionScaleFactors=(PositiveWidth=1,NegativeWidth=1,PositiveHeight=1,NegativeHeight=1)
    FalloffSize=0.2
    FlipHeightScaling=true
    OptionalParentItem=(Type=Bone,Name="spine_03")
    RotationOffset=(X=0,Y=30,Z=0)
    Debug.bDrawDebug=@n1.debug
    ExecutePin=@SequenceExecution.B
   )
   > Deform_GetMirrorTransform_FNC = RigVMFunctionReferenceNode(transform=@n3.Value.lat_45_#14)
   > &RigUnit_ModifyTransforms(
    ItemToModify=((Item=(Type=Bone,Name="deform_lat_r"),Transform=(Rotation=(X=0,Y=0,Z=0,W=1),Translation=(X=0,Y=0,Z=0),Scale3D=(X=1,Y=1,Z=1))))
    ItemToModify.0.Transform=@Deform_GetMirrorTransform_FNC.mirrorTransform
    Weight=@n4.OutputParam
   )
   RigVMFunctionReturnNode(ExecuteContext=@SequenceExecution.C)
  ]}
 ContainedGraph {nodes [
   n1 = RigVMFunctionEntryNode()
   SequenceExecution = &RigVMFunction_Sequence(ExecuteContext=@n1.ExecuteContext)
   SequenceExecution_1 = &RigVMFunction_Sequence(ExecuteContext=@SequenceExecution.B)
   SequenceExecution_2 = &RigVMFunction_Sequence(ExecuteContext=@SequenceExecution_1.B)
   RigVMFunctionReturnNode(A=@SequenceExecution.A, B=@SequenceExecution_1.A, C=@SequenceExecution_2.A, D=@SequenceExecution_2.B)
  ]}
 ContainedGraph {nodes [
   RigVMFunctionEntryNode()
   > SequenceExecution = RigVMAggregateNode()
   RigVMFunctionReturnNode(ExecuteContext=@SequenceExecution.D)
   n1 = /Script/ControlRig.RigUnit_AimBoneMath(
    DebugSettings=(Scale=10,WorldOffset=(Rotation=(X=0,Y=0,Z=0,W=1),Translation=(X=0,Y=0,Z=0),Scale3D=(X=1,Y=1,Z=1)))
    Primary=(Weight=1,Axis=(X=-1,Y=0,Z=0),Target=(X=0,Y=0,Z=0),Kind=Location,Space=(Type=Bone,Name="upperarm_r"))
    Secondary=(Weight=1,Axis=(X=0,Y=1,Z=0),Target=(X=0,Y=0,Z=0),Kind=Location,Space=(Type=Bone,Name="spine_04"))
    Weight=1
    InputTransform=&RigUnit_GetTransform(Item=(Type=Bone,Name="deform_pec_aim_r")).Transform
   ) {is=struct RigVMUnitNode}
   &RigUnit_SetTransform(Item=(Type=Bone,Name="deform_pec_aim_r"), Weight=1, bInitial=false, Value=@n1.Result, ExecutePin=@SequenceExecution.B)
   > /Script/ControlRig.RigUnit_PositionConstraintLocalSpaceOffset(
    Child=(Type=Bone,Name="deform_pec_r")
    Filter=(bX=True,bY=false,bZ=false)
    Parents=((Item=(Type=Bone,Name="deform_pec_base_r"),Weight=0),(Item=(Type=Bone,Name="upperarm_r"),Weight=1))
    Weight=1
    bMaintainOffset=True
   ) {is=struct RigVMUnitNode}
   > GetDeltaTransformFromRefPose_FNC_1 = RigVMFunctionReferenceNode(Item=(Type=Bone,Name="clavicle_r"))
   n2 = RigVMVariableNode() {variable=deform_value}
   > Deform_GetMirrorTransform_FNC = RigVMFunctionReferenceNode(transform=@n2.Value.pec_51_#15)
   > &RigUnit_ModifyTransforms(
    ItemToModify=((Item=(Type=Bone,Name="deform_pec_r"),Transform=(Rotation=(X=0,Y=0,Z=0,W=1),Translation=(X=0,Y=0,Z=0),Scale3D=(X=1,Y=1,Z=1))))
    ItemToModify.0.Transform=@Deform_GetMirrorTransform_FNC.mirrorTransform
    Weight=&RigVMFunction_MathFloatRemap(SourceMaximum=30, bClamp=true, Value=@GetDeltaTransformFromRefPose_FNC_1.OutTransform.Rotation.Pitch).Result
   )
   n3 = &RigUnit_AimBoneMath(
    DebugSettings=(Scale=10,WorldOffset=(Rotation=(X=0,Y=0,Z=0,W=1),Translation=(X=0,Y=0,Z=0),Scale3D=(X=1,Y=1,Z=1)))
    Primary=(Weight=1,Axis=(X=1,Y=0,Z=0),Target=(X=0,Y=0,Z=0),Kind=Location,Space=(Type=Bone,Name="upperarm_l"))
    Secondary=(Weight=1,Axis=(X=0,Y=-1,Z=0),Target=(X=0,Y=0,Z=0),Kind=Location,Space=(Type=Bone,Name="spine_04"))
    Weight=1
    InputTransform=&RigUnit_GetTransform(Item=(Type=Bone,Name="deform_pec_aim_l")).Transform
   )
   &RigUnit_SetTransform(Item=(Type=Bone,Name="deform_pec_aim_l"), Weight=1, bInitial=false, Value=@n3.Result, ExecutePin=@SequenceExecution.A)
   > &RigUnit_PositionConstraintLocalSpaceOffset(
    Child=(Type=Bone,Name="deform_pec_l")
    Filter=(bX=True,bY=false,bZ=false)
    Parents=((Item=(Type=Bone,Name="deform_pec_base_l"),Weight=0),(Item=(Type=Bone,Name="upperarm_l"),Weight=1))
    Weight=1
    bMaintainOffset=True
   )
   > GetDeltaTransformFromRefPose_FNC = RigVMFunctionReferenceNode(Item=(Type=Bone,Name="clavicle_l"))
   > &RigUnit_ModifyTransforms(
    ItemToModify=((Item=(Type=Bone,Name="deform_pec_l"),Transform=(Rotation=(X=0,Y=0,Z=0,W=1),Translation=(X=0,Y=0,Z=0),Scale3D=(X=1,Y=1,Z=1))))
    Weight=&RigVMFunction_MathFloatRemap(SourceMaximum=30, bClamp=true, Value=@GetDeltaTransformFromRefPose_FNC.OutTransform.Rotation.Pitch).Result
    ItemToModify.0.Transform=@n2.Value.pec_51_#15
   )
   &RigUnit_ParentConstraint(
    AdvancedSettings=(RotationOrderForFilter=XZY)
    Child=(Type=Bone,Name="deform_pec_aim_base_l")
    ExecutePin=()
    Parents=((Item=(Type=Bone,Name="spine_03"),Weight=0.25),(Item=(Type=Bone,Name="spine_04"),Weight=0.55),(Item=(Type=Bone,Name="spine_05"),Weight=0.2))
    Weight=1
   )
   &RigUnit_ParentConstraint(
    AdvancedSettings=(RotationOrderForFilter=XZY)
    Child=(Type=Bone,Name="deform_pec_aim_base_r")
    ExecutePin=()
    Parents=((Item=(Type=Bone,Name="spine_03"),Weight=0.25),(Item=(Type=Bone,Name="spine_04"),Weight=0.55),(Item=(Type=Bone,Name="spine_05"),Weight=0.2))
    Weight=1
   )
  ]}
 ContainedGraph {nodes [
   n1 = RigVMFunctionEntryNode()
   SequenceExecution = &RigVMFunction_Sequence(ExecuteContext=@n1.ExecuteContext)
   SequenceExecution_1 = &RigVMFunction_Sequence(ExecuteContext=@SequenceExecution.B)
   SequenceExecution_2 = &RigVMFunction_Sequence(ExecuteContext=@SequenceExecution_1.B)
   RigVMFunctionReturnNode(A=@SequenceExecution.A, B=@SequenceExecution_1.A, C=@SequenceExecution_2.A, D=@SequenceExecution_2.B)
  ]}
 ContainedGraph {nodes [
   n1 = RigVMFunctionEntryNode()
   n2 = &RigVMFunction_MathTransformToEulerTransform(Value=@n1.transform) {template="Cast::Execute(in Value,out Result)"}
   > RigVMFunctionReturnNode(
    mirrorTransform=(Rotation=(X=0,Y=0,Z=0,W=1),Translation=(X=0,Y=0,Z=0),Scale3D=(X=0,Y=0,Z=0))
    mirrorTransform.Rotation=/Script/RigVM.RigVMFunction_MathQuaternionFromRotatorV2(Value=@n2.Result.Rotation) {is=struct RigVMUnitNode, template="Cast::Execute(in Value,out Result)"}.Result
    mirrorTransform.Translation=/Script/RigVM.RigVMFunction_MathVectorScale(Factor=-1, Value=@n2.Result.Location) {is=struct RigVMUnitNode, template="Scale::Execute(in Value,in Factor,out Result)"}.Result
    mirrorTransform.Scale3D=@n2.Result.Scale
   )
  ]}
 RigVMFunctionLibrary {nodes [
   Deform_Elbow_FNC = RigVMCollapseNode(ExecuteContext=(), bicep=true, elbow_in=true, elbow_out=true)
   Deform_GetMirrorTransform_FNC = RigVMCollapseNode(ExecuteContext=(), transform=(Rotation=(X=0,Y=0,Z=0,W=1),Translation=(X=0,Y=0,Z=0),Scale3D=(X=1,Y=1,Z=1)))
   Deform_GroinGlue_FNC = RigVMCollapseNode(ExecuteContext=(), debug=False, glute=true, groin=true)
   Deform_Knee_FNC = RigVMCollapseNode(ExecuteContext=(), knee_in=true, knee_out=true)
   Deform_Lat_FNC = RigVMCollapseNode(ExecuteContext=(), debug=False)
   Deform_Pec_FNC = RigVMCollapseNode(ExecuteContext=())
   Deform_UpperArm_FNC = RigVMCollapseNode(ExecuteContext=(), debug_bk=false, debug_fr=false, debug_up=false, upperarm_bk=true, upperarm_fr=true, upperarm_up=true)
   Deform_Wrist_FNC = RigVMCollapseNode(ExecuteContext=(), debug_dn=false, debug_up=false, wrist_dn=true, wrist_up=true)
   GetDeltaTransformFromRefPose_FNC = RigVMCollapseNode(ExecuteContext=(), Item=(Type=Bone))
  ]}
 RigVMModel {nodes [
   n1 = RigVMRerouteNode(Value=())
   n2 = &RigVMFunction_ControlFlowBranch(ExecuteContext=@n1.Value, Condition=RigVMVariableNode() {variable=deform_bool}.Value.wrist_7_#16)
   Deform_Wrist_FNC_1 = RigVMFunctionReferenceNode(debug_dn=false, debug_up=false, wrist_dn=true, wrist_up=true, ExecuteContext=@n2.True)
   n3 = RigVMRerouteNode(Value=())
   n4 = &RigVMFunction_ControlFlowBranch(ExecuteContext=@n3.Value, Condition=RigVMVariableNode() {variable=deform_bool}.Value.pec_12_#17)
   Deform_Pec_FNC_1 = RigVMFunctionReferenceNode(ExecuteContext=@n4.True)
   n5 = RigVMRerouteNode(Value=())
   n6 = &RigVMFunction_ControlFlowBranch(ExecuteContext=@n5.Value, Condition=RigVMVariableNode() {variable=deform_bool}.Value.lat_15_#18)
   Deform_Lat_FNC_1 = RigVMFunctionReferenceNode(debug=False, ExecuteContext=@n6.True)
   RigUnit_BeginExecution = &RigUnit_BeginExecution()
   SequenceAggregate = &RigVMFunction_Sequence(ExecuteContext=@RigUnit_BeginExecution.ExecutePin)
   SequenceAggregate_1 = RigVMAggregateNode(ExecuteContext=@SequenceAggregate.A)
   n7 = &RigVMFunction_ControlFlowBranch(
    Condition=RigVMVariableNode() {variable=deform_bool}.Value.upperarm_1_#19
    ExecuteContext=@SequenceAggregate_1.A
   )
   Deform_UpperArm_FNC_1 = RigVMFunctionReferenceNode(debug_bk=false, debug_fr=false, debug_up=false, upperarm_bk=true, upperarm_fr=true, upperarm_up=true, ExecuteContext=@n7.True)
   n8 = &RigVMFunction_ControlFlowBranch(Condition=RigVMVariableNode() {variable=deform_bool}.Value.knee_9_#20, ExecuteContext=@SequenceAggregate_1.C)
   Deform_Knee_FNC_1 = RigVMFunctionReferenceNode(knee_in=true, knee_out=true, ExecuteContext=@n8.True)
   SequenceAggregate_2 = RigVMAggregateNode(ExecuteContext=@SequenceAggregate.B)
   n9 = RigVMVariableNode() {variable=deform_bool}
   n10 = &RigVMFunction_ControlFlowBranch(
    Condition=&RigVMFunction_MathBoolOr(B=@n9.Value.bicep_13_#21, A=@n9.Value.elbow_3_#22).Result
    ExecuteContext=@SequenceAggregate_2.B
   )
   Deform_Elbow_FNC_1 = RigVMFunctionReferenceNode(
    ExecuteContext=@n10.True
    bicep=@n9.Value.bicep_13_#21
    elbow_in=@n9.Value.elbow_3_#22
    elbow_out=@n9.Value.elbow_3_#22
   )
   n11 = RigVMVariableNode() {variable=deform_bool}
   n12 = &RigVMFunction_ControlFlowBranch(
    Condition=&RigVMFunction_MathBoolOr(B=@n11.Value.glute_17_#23, A=@n11.Value.groin_8_#24).Result
    ExecuteContext=@SequenceAggregate_2.C
   )
   Deform_GroinGlue_FNC_1 = RigVMFunctionReferenceNode(
    debug=False
    ExecuteContext=@n12.True
    glute=@n11.Value.glute_17_#23
    groin=@n11.Value.groin_8_#24
   )
  ]}
]
hierarchy [
 root {loc=<0 0 0>, rot=<0 0 0 1>}
 pelvis {parent=root, loc=<0 -0.7071 93.1191>, rot=<0.0127 0.707 -0.0127 -0.707>}
 spine_01 {parent=pelvis, loc=<6.7494 -0.6388 0>, rot=<0 0 0.034 -0.9994>}
 spine_02 {parent=spine_01, loc=<7.2518 -0.035 0>, rot=<0 0 0.0238 -0.9997>}
 spine_03 {parent=spine_02, loc=<6.0104 0 0>, rot=<0 0 0.0149 -0.9999>}
 spine_04 {parent=spine_03, loc=<8.1024 0 0>, rot=<0 0 -0.0132 -0.9999>}
 spine_05 {parent=spine_04, loc=<10.0146 0 0>, rot=<0 0 -0.0015 -1>}
 clavicle_l {parent=spine_05, loc=<4.3575 -0.1466 -2.3223>, rot=<0.0465 -0.7289 -0.009 -0.6829>}
 upperarm_l {parent=clavicle_l, loc=<15.5111 0 0>, rot=<0.0226 -0.3234 -0.0088 -0.9459>}
 lowerarm_l {parent=upperarm_l, loc=<26.6514 0 0>, rot=<0.0391 -0.0099 -0.2453 -0.9686>}
 hand_l {parent=lowerarm_l, loc=<22.3491 0 0>, rot=<-0.7081 -0.0507 0.0975 -0.6975>}
 index_metacarpal_l {parent=hand_l, loc=<1.6535 0.0183 -2.96>, rot=<0.0158 -0.0124 -0.0173 -0.9996>}
 index_01_l {parent=index_metacarpal_l, loc=<8.3008 0.0357 -0.0524>, rot=<-0.0044 0.0485 0.0924 -0.9945>}
 index_02_l {parent=index_01_l, loc=<3.3587 -0.3004 0.0452>, rot=<-0.002 0.0027 0.3225 -0.9466>}
 index_03_l {parent=index_02_l, loc=<2.5578 -0.123 -0.0252>, rot=<-0.002 0.0067 0.2558 -0.9667>}
 middle_metacarpal_l {parent=hand_l, loc=<2.1856 -0.1725 -0.5984>, rot=<0.006 0.0168 -0.0329 -0.9993>}
 middle_01_l {parent=middle_metacarpal_l, loc=<7.3275 -0.2447 0.0129>, rot=<-0.0208 0.0042 0.0858 -0.9961>}
 middle_02_l {parent=middle_01_l, loc=<4.0123 0.1401 0.1175>, rot=<-0.0218 0.0258 0.3512 -0.9357>}
 middle_03_l {parent=middle_02_l, loc=<2.6077 0.0534 0.1373>, rot=<0.0082 0.0188 0.1708 -0.9851>}
 const_middle_01_l {parent=middle_metacarpal_l, loc=<7.9682 -0.4541 0.8652>, rot=<-0.0537 0.0585 0.0816 -0.9935>}
 const_middle_02_l {parent=const_middle_01_l, loc=<2.7673 0 0>, rot=<0 0 0.3363 -0.9418>}
 const_middle_03_l {parent=const_middle_02_l, loc=<2.2699 0 0>, rot=<0.0082 0.0188 0.1708 -0.9851>}
 pinky_metacarpal_l {parent=hand_l, loc=<1.7616 -0.9753 3.3814>, rot=<-0.0397 0.0814 -0.069 -0.9935>}
 pinky_01_l {parent=pinky_metacarpal_l, loc=<7.1709 -0.4831 -0.1094>, rot=<-0.0654 -0.0035 0.2242 -0.9723>}
 pinky_02_l {parent=pinky_01_l, loc=<2.8274 0.0893 -0.1043>, rot=<0.0037 0.0011 0.2854 -0.9584>}
 pinky_03_l {parent=pinky_02_l, loc=<2.2046 0.0307 -0.0078>, rot=<-0.0002 0.0098 0.1445 -0.9895>}
 ring_metacarpal_l {parent=hand_l, loc=<2.2074 -0.3625 1.3638>, rot=<-0.0284 0.0722 -0.0458 -0.9959>}
 ring_01_l {parent=ring_metacarpal_l, loc=<7.0295 -0.6197 -0.1446>, rot=<-0.0489 -0.0252 0.1892 -0.9804>}
 ring_02_l {parent=ring_01_l, loc=<3.6174 0.4647 -0.1027>, rot=<0 0 0.2806 -0.9598>}
 ring_03_l {parent=ring_02_l, loc=<2.4152 0.114 0.0199>, rot=<0 0 0.1189 -0.9929>}
 thumb_01_l {parent=hand_l, loc=<1.3118 -0.3191 -2.459>, rot=<0.6546 -0.4042 0.0664 -0.6354>}
 thumb_02_l {parent=thumb_01_l, loc=<5.9336 -0.0448 0.0401>, rot=<-0.0037 0.0213 0.298 -0.9543>}
 thumb_03_l {parent=thumb_02_l, loc=<2.2223 0.3682 -0.0907>, rot=<0.0118 -0.013 0.0066 -0.9998>}
 weapon_l {parent=hand_l, loc=<-0.1974 -3.4111 0.3811>, rot=<0 0 0 1>}
 lowerarm_twist_01_l {parent=lowerarm_l, loc=<17.8549 0 0>, rot=<0 0 0 1>}
 lowerarm_twist_02_l {parent=lowerarm_l, loc=<9.6199 0 0>, rot=<0 0 0 1>}
 upperarm_twist_01_l {parent=upperarm_l, loc=<0.0041 0 0>, rot=<0 0 0 1>}
 deform_upperarm_bk_base_l {parent=upperarm_twist_01_l, loc=<-3.2582 -5.546 -2.0366>, rot=<0 -0.0007 -0.0191 -0.9998>}
 deform_upperarm_bk_l {parent=deform_upperarm_bk_base_l, loc=<0 0 0>, rot=<0 0 0 1>}
 deform_upperarm_fr_base_l {parent=upperarm_twist_01_l, loc=<-2.8462 3.9433 -2.0158>, rot=<0 -0.0007 -0.0191 -0.9998>}
 deform_upperarm_fr_l {parent=deform_upperarm_fr_base_l, loc=<0 0 0>, rot=<0 0 0 1>}
 deform_upperarm_up_base_l {parent=upperarm_twist_01_l, loc=<-5.9223 -0.7874 1.605>, rot=<0 -0.0007 -0.0191 -0.9998>}
 deform_upperarm_up_l {parent=deform_upperarm_up_base_l, loc=<0 0 0>, rot=<0 0 0 1>}
 upperarm_twist_02_l {parent=upperarm_l, loc=<13.7941 0 0>, rot=<0 0 0 1>}
 clavicle_r {parent=spine_05, loc=<4.3575 -0.1466 2.3223>, rot=<0.7289 0.0465 -0.6829 0.009>}
 upperarm_r {parent=clavicle_r, loc=<-15.5111 0 0>, rot=<0.0226 -0.3234 -0.0088 -0.9459>}
 lowerarm_r {parent=upperarm_r, loc=<-26.6515 0 0>, rot=<0.0391 -0.0099 -0.2453 -0.9686>}
 hand_r {parent=lowerarm_r, loc=<-22.3492 0 0>, rot=<-0.7081 -0.0507 0.0975 -0.6975>}
 index_metacarpal_r {parent=hand_r, loc=<-1.6534 -0.0181 2.9601>, rot=<0.0158 -0.0124 -0.0173 -0.9996>}
 index_01_r {parent=index_metacarpal_r, loc=<-8.3008 -0.0357 0.0524>, rot=<-0.0044 0.0485 0.0924 -0.9945>}
 index_02_r {parent=index_01_r, loc=<-3.3589 0.3003 -0.0453>, rot=<-0.002 0.0027 0.3225 -0.9466>}
 index_03_r {parent=index_02_r, loc=<-2.5578 0.123 0.0252>, rot=<-0.002 0.0067 0.2558 -0.9667>}
 middle_metacarpal_r {parent=hand_r, loc=<-2.1856 0.1726 0.5984>, rot=<0.006 0.0168 -0.0329 -0.9993>}
 middle_01_r {parent=middle_metacarpal_r, loc=<-7.3273 0.2447 -0.0129>, rot=<-0.0208 0.0042 0.0858 -0.9961>}
 middle_02_r {parent=middle_01_r, loc=<-4.0142 -0.1411 -0.07>, rot=<-0.0217 0.0259 0.3512 -0.9357>}
 middle_03_r {parent=middle_02_r, loc=<-2.6074 -0.0534 -0.1312>, rot=<0.0083 0.0188 0.1708 -0.9851>}
 const_middle_01_r {parent=middle_metacarpal_r, loc=<-7.9681 0.4541 -0.8652>, rot=<-0.0537 0.0585 0.0816 -0.9935>}
 const_middle_02_r {parent=const_middle_01_r, loc=<-2.7672 0 0>, rot=<0 0 0.3363 -0.9418>}
 const_middle_03_r {parent=const_middle_02_r, loc=<-2.2699 0 0>, rot=<0.0082 0.0188 0.1708 -0.9851>}
 pinky_metacarpal_r {parent=hand_r, loc=<-1.7617 0.9753 -3.3814>, rot=<-0.0397 0.0814 -0.069 -0.9935>}
 pinky_01_r {parent=pinky_metacarpal_r, loc=<-7.1707 0.4832 0.1094>, rot=<-0.0654 -0.0035 0.2242 -0.9723>}
 pinky_02_r {parent=pinky_01_r, loc=<-2.8274 -0.0892 0.1043>, rot=<0.0037 0.0011 0.2854 -0.9584>}
 pinky_03_r {parent=pinky_02_r, loc=<-2.2045 -0.031 0.0078>, rot=<-0.0002 0.0098 0.1445 -0.9895>}
 ring_metacarpal_r {parent=hand_r, loc=<-2.2073 0.3627 -1.3638>, rot=<-0.0284 0.0722 -0.0458 -0.9959>}
 ring_01_r {parent=ring_metacarpal_r, loc=<-7.0297 0.6197 0.1446>, rot=<-0.0489 -0.0252 0.1892 -0.9804>}
 ring_02_r {parent=ring_01_r, loc=<-3.6171 -0.4648 0.1027>, rot=<0 0 0.2806 -0.9598>}
 ring_03_r {parent=ring_02_r, loc=<-2.4153 -0.1141 -0.0199>, rot=<0 0 0.1189 -0.9929>}
 thumb_01_r {parent=hand_r, loc=<-1.3119 0.3192 2.459>, rot=<0.6546 -0.4042 0.0664 -0.6354>}
 thumb_02_r {parent=thumb_01_r, loc=<-5.9335 0.0447 -0.0401>, rot=<-0.0037 0.0213 0.298 -0.9543>}
 thumb_03_r {parent=thumb_02_r, loc=<-2.2223 -0.3682 0.0907>, rot=<0.0118 -0.013 0.0066 -0.9998>}
 weapon_r {parent=hand_r, loc=<0.1975 3.4112 -0.3811>, rot=<0 0 0 -1>}
 lowerarm_twist_01_r {parent=lowerarm_r, loc=<-17.855 0 0>, rot=<0 0 0 1>}
 lowerarm_twist_02_r {parent=lowerarm_r, loc=<-9.6199 0 0>, rot=<0 0 0 1>}
 upperarm_twist_01_r {parent=upperarm_r, loc=<-0.0041 0 0>, rot=<0 0 0 1>}
 deform_upperarm_bk_base_r {parent=upperarm_twist_01_r, loc=<3.2581 5.546 2.0367>, rot=<0 -0.0007 -0.0191 -0.9998>}
 deform_upperarm_bk_r {parent=deform_upperarm_bk_base_r, loc=<0 0 0>, rot=<0 0 0 1>}
 deform_upperarm_fr_base_r {parent=upperarm_twist_01_r, loc=<2.8462 -3.9432 2.0159>, rot=<0 -0.0007 -0.0191 -0.9998>}
 deform_upperarm_fr_r {parent=deform_upperarm_fr_base_r, loc=<0 0 0>, rot=<0 0 0 1>}
 deform_upperarm_up_base_r {parent=upperarm_twist_01_r, loc=<5.9222 0.7874 -1.6049>, rot=<0 -0.0007 -0.0191 -0.9998>}
 deform_upperarm_up_r {parent=deform_upperarm_up_base_r, loc=<0 0 0>, rot=<0 0 0 1>}
 upperarm_twist_02_r {parent=upperarm_r, loc=<-13.7942 0 0>, rot=<0 0 0 1>}
 neck_01 {parent=spine_05, loc=<11.3138 0.1529 0>, rot=<0 0 -0.2003 -0.9797>}
 neck_02 {parent=neck_01, loc=<4.5627 0 0>, rot=<0 0 0.0504 -0.9987>}
 head {parent=neck_02, loc=<5.3914 0 0>, rot=<0 0 0.115 -0.9934>}
 thigh_l {parent=pelvis, loc=<-2.5693 -0.0109 -7.7501>, rot=<0.0012 0.0572 0.0155 -0.9982>}
 calf_l {parent=thigh_l, loc=<-39.0484 0 0>, rot=<0.0006 -0.0161 -0.0391 -0.9991>}
 calf_twist_01_l {parent=calf_l, loc=<-35.813 -0.0967 0.0397>, rot=<0.0237 0.0095 -0.0035 -0.9997>}
 calf_twist_02_l {parent=calf_l, loc=<-18.8064 -0.0967 0.0397>, rot=<0.0237 0.0095 -0.0035 -0.9997>}
 foot_l {parent=calf_l, loc=<-42.561 0.1077 0.4292>, rot=<0.0418 -0.0455 0.0343 -0.9975>}
 ball_l {parent=foot_l, loc=<-6.1599 13.8264 0.6806>, rot=<0.0001 -0.0001 -0.7063 -0.7079>}
 deform_knee_in_base_l {parent=calf_l, loc=<3.6268 -5.2446 -0.2814>, rot=<-0.0017 -0.0216 -0.0001 -0.9998>}
 deform_knee_in_l {parent=deform_knee_in_base_l, loc=<0 0 0>, rot=<0 0 0 1>}
 deform_knee_out_base_l {parent=calf_l, loc=<2.8134 3.7791 -0.4614>, rot=<-0.0017 -0.0216 -0.0001 -0.9998>}
 deform_knee_out_l {parent=deform_knee_out_base_l, loc=<0 0 0>, rot=<0 0 0 1>}
 dyn_knee_L {parent=calf_l, loc=<-9.2527 -6.7873 0>, rot=<0.01 0.0235 -0.7094 -0.7044>}
 thigh_twist_01_l {parent=thigh_l, loc=<-21.6303 0 0>, rot=<0.0013 0 0 -1>}
 thigh_r {parent=pelvis, loc=<-2.5691 -0.0109 7.7501>, rot=<0.0572 -0.0012 0.9982 0.0155>}
 calf_r {parent=thigh_r, loc=<39.0487 0 0>, rot=<0.0006 -0.0161 -0.0391 -0.9991>}
 calf_twist_01_r {parent=calf_r, loc=<35.8132 0.0967 -0.0397>, rot=<0.0237 0.0095 -0.0035 -0.9997>}
 calf_twist_02_r {parent=calf_r, loc=<18.8064 0.0967 -0.0397>, rot=<0.0237 0.0095 -0.0035 -0.9997>}
 foot_r {parent=calf_r, loc=<42.5612 -0.1077 -0.4292>, rot=<0.0418 -0.0455 0.0343 -0.9975>}
 ball_r {parent=foot_r, loc=<6.1598 -13.8264 -0.6806>, rot=<0.0001 -0.0001 -0.7063 -0.7079>}
 deform_knee_in_base_r {parent=calf_r, loc=<-3.6267 5.2446 0.2813>, rot=<-0.0017 -0.0216 -0.0001 -0.9998>}
 deform_knee_in_r {parent=deform_knee_in_base_r, loc=<0 0 0>, rot=<0 0 0 1>}
 deform_knee_out_base_r {parent=calf_r, loc=<-2.8134 -3.7791 0.4614>, rot=<-0.0017 -0.0216 -0.0001 -0.9998>}
 deform_knee_out_r {parent=deform_knee_out_base_r, loc=<0 0 0>, rot=<0 0 0 1>}
 dyn_Knee_R {parent=calf_r, loc=<8.4765 5.7458 0>, rot=<0.01 0.0235 -0.7094 -0.7044>}
 thigh_twist_01_r {parent=thigh_r, loc=<21.6303 0 0>, rot=<0.0013 0 0 -1>}
 dyn_coat_MSTR_frnt_r {parent=pelvis, loc=<11.3406 7.5127 7.226>, rot=<0.1203 -0.2308 0.9616 0.0866>}
 dyn_coat_frnt_r {parent=dyn_coat_MSTR_frnt_r, loc=<0.0674 0 0>, rot=<0 0 0 1>}
 dyn_coat_frnt_01_r {parent=dyn_coat_frnt_r, loc=<9.9128 0 0>, rot=<0 0 0 1>}
 dyn_coat_MSTR_frnt_l {parent=pelvis, loc=<6.0124 8.8592 -5.8759>, rot=<-0.1203 0.2308 0.9616 0.0866>}
 dyn_coat_frnt_l {parent=dyn_coat_MSTR_frnt_l, loc=<0.0674 0 0>, rot=<0 0 0 1>}
 dyn_coat_frnt_01_l {parent=dyn_coat_frnt_l, loc=<9.9128 0 0>, rot=<0 0 0 1>}
 dyn_coat_MSTR_back_l {parent=pelvis, loc=<5.615 -10.6358 -7.0829>, rot=<0.0621 -0.9636 -0.2499 -0.0719>}
 dyn_coat_back_l {parent=dyn_coat_MSTR_back_l, loc=<0.0674 0 0>, rot=<0 0 0 1>}
 dyn_coat_back_01_l {parent=dyn_coat_back_l, loc=<9.3825 0 0>, rot=<0 0 0 1>}
 dyn_coat_MSTR_back_r {parent=pelvis, loc=<5.615 -10.6358 6.9916>, rot=<0.0623 -0.9649 0.2446 0.0723>}
 dyn_coat_back_r {parent=dyn_coat_MSTR_back_r, loc=<0.0674 0 0>, rot=<0 0 0 1>}
 dyn_coat_back_01_r {parent=dyn_coat_back_r, loc=<9.3825 0 0>, rot=<0 0 0 1>}
 dyn_coat_MSTR_side_r {parent=pelvis, loc=<8.074 -0.241 14.225>, rot=<0.0622 -0.7046 0.7019 0.0839>}
 dyn_coat_side_r {parent=dyn_coat_MSTR_side_r, loc=<0.0674 0 0>, rot=<0 0 0 1>}
 dyn_coat_side_01_r {parent=dyn_coat_side_r, loc=<8.3923 0 0>, rot=<0 0 0 1>}
 dyn_coat_MSTR_side_l {parent=pelvis, loc=<8.074 -0.241 -14.2245>, rot=<0.0651 -0.7039 -0.702 -0.0869>}
 dyn_coat_side_l {parent=dyn_coat_MSTR_side_l, loc=<0.0674 0 0>, rot=<0 0 0 1>}
 dyn_coat_side_01_l {parent=dyn_coat_side_l, loc=<8.3923 0 0>, rot=<0 0 0 1>}
 dyn_ass {parent=pelvis, loc=<-10.1133 -4.6149 -5.5076>, rot=<0.0513 -0.9915 -0.097 -0.0699>}
 dyn_coat_MSTR_back_MID_ {parent=pelvis, loc=<5.5792 -11.6351 -0.1084>, rot=<0.0768 -0.9969 -0.0116 -0.0081>}
 dyn_coat_back_MID {parent=dyn_coat_MSTR_back_MID_, loc=<0.2474 -0.5266 0.172>, rot=<0 0 0 1>}
 dyn_MSTR_side_frt_R {parent=pelvis, loc=<7.9348 4.374 13.1528>, rot=<-0.0577 -0.3028 0.9404 0.1438>}
 dyn_side_frt_R {parent=dyn_MSTR_side_frt_R, loc=<0.0674 0 0>, rot=<0 0 0 1>}
 dyn_side_frt_R_01 {parent=dyn_side_frt_R, loc=<8.3923 0 0>, rot=<0 0 0 1>}
 dyn_MSTR_frnt_side_l {parent=pelvis, loc=<8.0411 4.039 -12.1309>, rot=<0.0366 0.355 0.9202 0.1607>}
 dyn_frn_sidet_l {parent=dyn_MSTR_frnt_side_l, loc=<0.0674 0 0>, rot=<0 0 0 1>}
 dyn_frnt_side_01_l {parent=dyn_frn_sidet_l, loc=<9.9128 0 0>, rot=<0 0 0 1>}
 dyn_pelvis_scale {parent=pelvis, loc=<0 0 0>, rot=<0 0 0 1>, scale=<0.9353 0.9353 0.9353>}
 dyn_glute {parent=pelvis, loc=<-10.1133 -4.6149 -5.5076>, rot=<0.0513 -0.9915 -0.097 -0.0699>}
 attach {parent=root, loc=<0 0 0>, rot=<0 0 0 1>}
 ik_foot_root {parent=root, loc=<0 0 0>, rot=<0 0 0 1>}
 ik_foot_l {parent=ik_foot_root, loc=<15.2608 -4.4176 9.4433>, rot=<0.0357 0.7045 0.0284 -0.7082>}
 ik_foot_r {parent=ik_foot_root, loc=<-15.2609 -4.4176 9.4431>, rot=<0.7082 0.0284 -0.7045 0.0357>}
 ik_hand_root {parent=root, loc=<0 0 0>, rot=<0 0 0 1>}
 ik_hand_gun {parent=ik_hand_root, loc=<-51.9729 4.0884 102.6935>, rot=<0.6084 0.1793 0.4613 -0.6205>}
 ik_hand_l {parent=ik_hand_gun, loc=<53.0286 82.1782 35.2051>, rot=<-0.5102 -0.7906 -0.3387 0>}
 ik_hand_r {parent=ik_hand_gun, loc=<0 0 0>, rot=<0 0 0 1>}
]
variables [
 deform_bool {
  type=FUserDefinedStruct_#25
  default="(upperarm_1_#19=True,elbow_3_#22=True,wrist_7_#16=False,groin_8_#24=True,glute_17_#23=True,knee_9_40C408324CA26 ...(truncated)"
 }
 deform_value {
  type=FUserDefinedStruct_#26
  default="(upperarm_up_18_#1=(Rotation=(X=0,Y=0,Z=0,W=1),Translation=(X=0,Y=0,Z=8),Scale3D=(X=0.7,Y=1,Z=1)),upperarm_fwd_fr_19_#2=(Rotation=(X=0,Y=0,Z=0,W=1),Translation=(X=0,Y=10,Z=0),Scale3D=(X=0.5,Y=1,Z= ...(truncated)"
 }
]
