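# Physics asset for the Henchman_Bad2 player body's coat, named as the crouch variant, covering the coat bones plus pelvis and thighs.
# Coat master bones and thigh bodies lock angularly to the pelvis, whose capsule and the thigh shapes form the body volume.
# Back, front and side segments chain from the masters with small swing angles and often locked twist.
# An extra locked constraint ties the left side panel to the left front-side panel.
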PhysicsAsset /Game/Characters/Player/Male/Medium/Bodies/M_MED_Henchman_Bad/Meshes/M_MED_Henchman_Bad2_CrouchPhysics
bodies {
 dyn_MSTR_frnt_side_l [{shapeType=Capsule, length=0.505, radius=0.505, shapeName=dyn_MSTR_frnt_side_l_capsule, rotation=<rot 0 -90 0>}]
 dyn_MSTR_side_frt_R [{shapeType=Capsule, length=0.505, radius=0.505, shapeName=dyn_MSTR_side_frt_R_capsule, rotation=<rot 0 -90 0>}]
 dyn_coat_MSTR_back_l [{shapeType=Capsule, length=15.9, radius=0.5, shapeName=dyn_coat_MSTR_back_l_capsule, center=<0 -2.6129 0>}]
 dyn_coat_MSTR_back_r [{shapeType=Capsule, length=15.9, radius=0.5, shapeName=dyn_coat_MSTR_back_r_capsule, center=<0 -2.6129 0>}]
 dyn_coat_MSTR_frnt_l [{shapeType=Capsule, length=0, radius=0.5}]
 dyn_coat_MSTR_frnt_r [{shapeType=Capsule, length=0.505, radius=0.505, shapeName=dyn_coat_MSTR_frnt_r_capsule, rotation=<rot 0 -90 0>}]
 dyn_coat_MSTR_side_l [{shapeType=Capsule, length=0, radius=0.5}]
 dyn_coat_MSTR_side_r [{shapeType=Capsule, length=0.505, radius=0.505, shapeName=dyn_coat_MSTR_side_r_capsule, rotation=<rot 0 -90 0>}]
 dyn_coat_back_01_l [{shapeType=Capsule, length=14.9, radius=2.8, shapeName=dyn_coat_back_01_l_capsule}]
 dyn_coat_back_01_r [{shapeType=Capsule, length=14, radius=3.1}]
 dyn_coat_back_l [{shapeType=Capsule, length=13.8, radius=3, shapeName=dyn_coat_back_l_capsule}]
 dyn_coat_back_r [{shapeType=Capsule, length=14, radius=2.7}]
 dyn_coat_frnt_01_l [{
   shapeType=Box
   extentX=6.0755
   extentY=1.5
   extentZ=16.5
   shapeName=dyn_coat_frnt_01_l_box
   center=<2.6032 -1.1 0.0767>
   rotation=<rot 11.172 0.5929 4.2572>
  }]
 dyn_coat_frnt_01_r [{
   shapeType=Box
   extentX=9.9
   extentY=2.7
   extentZ=15.9
   shapeName=dyn_coat_frnt_01_r_box
   center=<6.3296 0 0>
   rotation=<rot -6.7193 7.2269 -0.85>
  }]
 dyn_coat_frnt_l [{
   shapeType=Box
   extentX=9.6419
   extentY=2.2557
   extentZ=14.5
   shapeName=dyn_coat_frnt_l_box
   center=<4.7336 0 0>
   rotation=<rot 0.0063 -0.0128 3.401>
  }]
 dyn_coat_frnt_r [{shapeType=Box, extentX=10.7, extentY=2.1, extentZ=12.5, shapeName=dyn_coat_frnt_r_box, center=<5.1871 0 0>}]
 dyn_coat_side_01_l [{shapeType=Box, extentX=7.5, extentY=3.1205, extentZ=12.5, shapeName=dyn_coat_side_01_l_box, center=<6.1906 1 0>}]
 dyn_coat_side_01_r [
  {shapeType=Capsule, length=0, radius=0, shapeName=dyn_coat_side_01_r_capsule}
  {shapeType=Box, extentX=8.5, extentY=4.1, extentZ=12.5, shapeName=dyn_coat_side_01_r_box, center=<4.5454 0.6575 0>}
 ]
 dyn_coat_side_l [{shapeType=Box, extentX=7.5, extentY=1.5, extentZ=8.5, shapeName=dyn_coat_side_l_box, center=<6.1906 0 0>, rotation=<rot 0.0002 0.012 0>}]
 dyn_coat_side_r [
  {shapeType=Capsule, length=0, radius=0, shapeName=dyn_coat_side_r_capsule}
  {shapeType=Box, extentX=8.5, extentY=2.5, extentZ=12.5, shapeName=dyn_coat_side_r_box, center=<4.502 0 0>}
 ]
 dyn_frn_sidet_l [{
   shapeType=Box
   extentX=12.3097
   extentY=1.5868
   extentZ=9.1
   shapeName=dyn_frn_sidet_l_box
   center=<4.157 -0.4 0>
   rotation=<rot -10.9392 2.981 -9.8388>
  }]
 dyn_side_frt_R [{
   shapeType=Box
   extentX=4.7
   extentY=8.7
   extentZ=12.5
   shapeName=dyn_side_frt_R_box
   center=<12.8508 0.576 2.9322>
   rotation=<rot 14.4773 3.9668 22.0694>
  }]
 pelvis [{shapeType=Capsule, length=9.8495, radius=9.8243, center=<-0.0062 -0.8187 -0.0691>}]
 thigh_twist_01_l [{shapeType=Capsule, length=32.4405, radius=9.4, center=<6.6273 -0.2243 0.7872>, rotation=<rot 90 0 0>}]
 thigh_twist_01_r [
  {shapeType=Capsule, length=27.4179, radius=9.4, center=<-8.9388 0.2243 -0.7872>, rotation=<rot 90 0 0>}
  {
   shapeType=Box
   extentX=17.7
   extentY=7.5
   extentZ=7.5
   shapeName=thigh_twist_01_r_box
   center=<-5.9543 -3.26 5.7908>
   rotation=<rot 0.2198 -8.8343 1.4141>
  }
 ]
}
constraints [
 {
  bone1Name=dyn_coat_MSTR_back_l
  bone2Name=pelvis
  swing1LimitDegrees=10
  swing1Motion=Locked
  swing2LimitDegrees=5
  swing2Motion=Locked
  twistLimitDegrees=1
  twistMotion=Locked
 }
 {
  bone1Name=dyn_coat_back_l
  bone2Name=dyn_coat_MSTR_back_l
  swing1LimitDegrees=15
  swing2LimitDegrees=5
  twistLimitDegrees=0
  twistMotion=Locked
 }
 {
  bone1Name=dyn_coat_back_01_l
  bone2Name=dyn_coat_back_l
  swing1LimitDegrees=15
  swing2LimitDegrees=5
  twistLimitDegrees=1
  twistMotion=Locked
 }
 {
  bone1Name=dyn_coat_MSTR_back_r
  bone2Name=pelvis
  swing1LimitDegrees=10
  swing1Motion=Locked
  swing2LimitDegrees=5
  swing2Motion=Locked
  twistLimitDegrees=1
  twistMotion=Locked
 }
 {
  bone1Name=dyn_coat_MSTR_frnt_l
  bone2Name=pelvis
  swing1LimitDegrees=45
  swing1Motion=Locked
  swing2LimitDegrees=10
  swing2Motion=Locked
  twistLimitDegrees=45
  twistMotion=Locked
 }
 {bone1Name=dyn_coat_frnt_l, bone2Name=dyn_coat_MSTR_frnt_l, swing1LimitDegrees=10, swing2LimitDegrees=6, twistLimitDegrees=1}
 {bone1Name=dyn_coat_frnt_01_l, bone2Name=dyn_coat_frnt_l, swing1LimitDegrees=10, swing2LimitDegrees=3, twistLimitDegrees=2}
 {
  bone1Name=dyn_coat_MSTR_side_l
  bone2Name=pelvis
  swing1LimitDegrees=45
  swing1Motion=Locked
  swing2LimitDegrees=10
  swing2Motion=Locked
  twistLimitDegrees=45
  twistMotion=Locked
 }
 {
  bone1Name=dyn_coat_side_l
  bone2Name=dyn_coat_MSTR_side_l
  swing1LimitDegrees=10
  swing2LimitDegrees=5
  twistLimitDegrees=45
  twistMotion=Locked
 }
 {
  bone1Name=dyn_coat_side_01_l
  bone2Name=dyn_coat_side_l
  swing1LimitDegrees=15
  swing2LimitDegrees=2
  twistLimitDegrees=45
  twistMotion=Locked
 }
 {
  bone1Name=dyn_coat_MSTR_frnt_r
  bone2Name=pelvis
  swing1LimitDegrees=45
  swing1Motion=Locked
  swing2LimitDegrees=10
  swing2Motion=Locked
  twistLimitDegrees=45
  twistMotion=Locked
 }
 {bone1Name=dyn_coat_frnt_r, bone2Name=dyn_coat_MSTR_frnt_r, swing1LimitDegrees=17, swing2LimitDegrees=3, twistLimitDegrees=3}
 {
  bone1Name=dyn_coat_MSTR_side_r
  bone2Name=pelvis
  swing1LimitDegrees=45
  swing1Motion=Locked
  swing2LimitDegrees=45
  swing2Motion=Locked
  twistLimitDegrees=45
  twistMotion=Locked
 }
 {bone1Name=dyn_coat_side_r, bone2Name=dyn_coat_MSTR_side_r, swing1LimitDegrees=15, swing2LimitDegrees=1, twistLimitDegrees=0}
 {
  bone1Name=dyn_coat_side_01_r
  bone2Name=dyn_coat_side_r
  swing1LimitDegrees=20
  swing2LimitDegrees=0
  twistLimitDegrees=45
  twistMotion=Locked
 }
 {
  bone1Name=dyn_MSTR_side_frt_R
  bone2Name=pelvis
  swing1LimitDegrees=45
  swing1Motion=Locked
  swing2LimitDegrees=45
  swing2Motion=Locked
  twistLimitDegrees=45
  twistMotion=Locked
 }
 {
  bone1Name=dyn_side_frt_R
  bone2Name=dyn_MSTR_side_frt_R
  swing1LimitDegrees=15
  swing2LimitDegrees=10
  twistLimitDegrees=45
  twistMotion=Locked
 }
 {
  bone1Name=dyn_MSTR_frnt_side_l
  bone2Name=pelvis
  swing1LimitDegrees=45
  swing1Motion=Locked
  swing2LimitDegrees=45
  swing2Motion=Locked
  twistLimitDegrees=45
  twistMotion=Locked
 }
 {
  bone1Name=thigh_twist_01_l
  bone2Name=pelvis
  swing1LimitDegrees=45
  swing1Motion=Locked
  swing2LimitDegrees=45
  swing2Motion=Locked
  twistLimitDegrees=45
  twistMotion=Locked
 }
 {
  bone1Name=thigh_twist_01_r
  bone2Name=pelvis
  swing1LimitDegrees=45
  swing1Motion=Locked
  swing2LimitDegrees=45
  swing2Motion=Locked
  twistLimitDegrees=45
  twistMotion=Locked
 }
 {bone1Name=dyn_frn_sidet_l, bone2Name=dyn_MSTR_frnt_side_l, swing1LimitDegrees=10, swing2LimitDegrees=5, twistLimitDegrees=0}
 {
  bone1Name=dyn_coat_side_l
  bone2Name=dyn_frn_sidet_l
  swing1LimitDegrees=45
  swing1Motion=Locked
  swing2LimitDegrees=45
  swing2Motion=Locked
  twistLimitDegrees=45
  twistMotion=Locked
 }
 {
  bone1Name=dyn_coat_back_r
  bone2Name=dyn_coat_MSTR_back_r
  swing1LimitDegrees=15
  swing2LimitDegrees=5
  twistLimitDegrees=45
  twistMotion=Locked
 }
 {
  bone1Name=dyn_coat_back_01_r
  bone2Name=dyn_coat_back_r
  swing1LimitDegrees=10
  swing2LimitDegrees=5
  twistLimitDegrees=45
  twistMotion=Locked
 }
 {bone1Name=dyn_coat_frnt_01_r, bone2Name=dyn_coat_frnt_r, swing1LimitDegrees=7, swing2LimitDegrees=2, twistLimitDegrees=10}
]
