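# Physics asset for the Henchman Bad 2 male player body that sets up dynamic coat panels around the hips and legs.
# Pelvis and thigh capsules form the body, with small master anchor capsules angularly locked to the pelvis.
# Each anchor leads a two-box chain for front, side, back or front-side panels, with per-joint swing and twist limits.
# Lower front boxes also lock angularly to lower side boxes, tying those panels together.
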PhysicsAsset /Game/Characters/Player/Male/Medium/Bodies/M_MED_Henchman_Bad/Meshes/M_MED_Henchman_Bad2_Physics
bodies {
 dyn_MSTR_frnt_side_l [{shapeType=Capsule, length=1, radius=1}]
 dyn_MSTR_side_frt_R [{shapeType=Capsule, length=1, radius=1}]
 dyn_coat_MSTR_back_l [{shapeType=Capsule, length=0.505, radius=0.505}]
 dyn_coat_MSTR_back_r [{shapeType=Capsule, length=0.505, radius=0.505}]
 dyn_coat_MSTR_frnt_l [{shapeType=Capsule, length=0.505, radius=0.505}]
 dyn_coat_MSTR_frnt_r [{shapeType=Capsule, length=0.505, radius=0.505}]
 dyn_coat_MSTR_side_l [{shapeType=Capsule, length=0.505, radius=0.505}]
 dyn_coat_MSTR_side_r [{shapeType=Capsule, length=0.505, radius=0.505}]
 dyn_coat_back_01_l [{shapeType=Box, extentX=10, extentY=2, extentZ=15, center=<5 0 0>}]
 dyn_coat_back_01_r [{shapeType=Box, extentX=10, extentY=2.3559, extentZ=15, center=<5 0 0>}]
 dyn_coat_back_l [{shapeType=Box, extentX=10, extentY=2, extentZ=15, center=<5 0 0>}]
 dyn_coat_back_r [{shapeType=Box, extentX=10, extentY=2.4652, extentZ=15, center=<5 0 0>}]
 dyn_coat_frnt_01_l [{shapeType=Box, extentX=10, extentY=2, extentZ=15, center=<5 -0.3097 0.0001>}]
 dyn_coat_frnt_01_r [{shapeType=Box, extentX=10, extentY=2.9692, extentZ=13.9906, center=<5.0118 0.0842 -0.0509>, rotation=<rot -0.2117 2.1334 -6.553>}]
 dyn_coat_frnt_l [{shapeType=Box, extentX=10, extentY=2, extentZ=15, center=<5 0.3269 0>}]
 dyn_coat_frnt_r [{shapeType=Box, extentX=8, extentY=2, extentZ=15, center=<5 0 0>}]
 dyn_coat_side_01_l [{shapeType=Box, extentX=10, extentY=3.2572, extentZ=16.7915, center=<5 -0.1701 -0.0272>, rotation=<rot 0.0117 -0.0008 -1.7225>}]
 dyn_coat_side_01_r [{shapeType=Box, extentX=10, extentY=2.753, extentZ=14, center=<5 0 0>}]
 dyn_coat_side_l [{shapeType=Box, extentX=10, extentY=2, extentZ=16.5957, center=<5 -0.939 -0.0594>, rotation=<rot -0.0002 -0.0009 -3.6259>}]
 dyn_coat_side_r [{shapeType=Box, extentX=10, extentY=2, extentZ=14, center=<5 -0.8497 0>}]
 dyn_frn_sidet_l [{shapeType=Box, extentX=7.5, extentY=2.5, extentZ=7.5, shapeName=dyn_frn_sidet_l_box, center=<2.8002 0 0>}]
 dyn_frnt_side_01_l [{shapeType=Box, extentX=9.5, extentY=2.5, extentZ=7.5, shapeName=dyn_frnt_side_01_l_box, center=<0.8783 0 0>}]
 dyn_side_frt_R [{shapeType=Box, extentX=7.5, extentY=1.5, extentZ=7.5, shapeName=dyn_side_frt_R_box, center=<4.5723 0 0>}]
 dyn_side_frt_R_01 [{shapeType=Box, extentX=7.5, extentY=4.5, extentZ=7.5, shapeName=dyn_side_frt_R_01_box, center=<4.3085 0 0>}]
 pelvis [{shapeType=Capsule, length=9.8495, radius=9.8243, center=<-0.0062 -0.8187 -0.0691>}]
 thigh_l [{shapeType=Capsule, length=16.7082, radius=9.8708, center=<-9.9949 -0.8043 0.2826>, rotation=<rot 0 89.9999 -88.5886>}]
 thigh_r [{shapeType=Capsule, length=11.0472, radius=9.4664, center=<7.7267 0.4487 -1.1981>, rotation=<rot 0 87.9613 -89.9999>}]
}
constraints [
 {
  bone1Name=thigh_l
  bone2Name=pelvis
  swing1LimitDegrees=45
  swing1Motion=Locked
  swing2LimitDegrees=45
  swing2Motion=Locked
  twistLimitDegrees=45
  twistMotion=Locked
 }
 {
  bone1Name=thigh_r
  bone2Name=pelvis
  swing1LimitDegrees=45
  swing1Motion=Locked
  swing2LimitDegrees=45
  swing2Motion=Locked
  twistLimitDegrees=45
  twistMotion=Locked
 }
 {
  bone1Name=dyn_coat_MSTR_frnt_r
  bone2Name=pelvis
  swing1LimitDegrees=45
  swing1Motion=Locked
  swing2LimitDegrees=45
  swing2Motion=Locked
  twistLimitDegrees=45
  twistMotion=Locked
 }
 {
  bone1Name=dyn_coat_MSTR_frnt_l
  bone2Name=pelvis
  swing1LimitDegrees=45
  swing1Motion=Locked
  swing2LimitDegrees=45
  swing2Motion=Locked
  twistLimitDegrees=45
  twistMotion=Locked
 }
 {
  bone1Name=dyn_coat_MSTR_back_l
  bone2Name=pelvis
  swing1LimitDegrees=45
  swing1Motion=Locked
  swing2LimitDegrees=45
  swing2Motion=Locked
  twistLimitDegrees=45
  twistMotion=Locked
 }
 {
  bone1Name=dyn_coat_MSTR_back_r
  bone2Name=pelvis
  swing1LimitDegrees=45
  swing1Motion=Locked
  swing2LimitDegrees=45
  swing2Motion=Locked
  twistLimitDegrees=45
  twistMotion=Locked
 }
 {
  bone1Name=dyn_coat_MSTR_side_r
  bone2Name=pelvis
  swing1LimitDegrees=45
  swing1Motion=Locked
  swing2LimitDegrees=45
  swing2Motion=Locked
  twistLimitDegrees=45
  twistMotion=Locked
 }
 {
  bone1Name=dyn_coat_MSTR_side_l
  bone2Name=pelvis
  swing1LimitDegrees=45
  swing1Motion=Locked
  swing2LimitDegrees=45
  swing2Motion=Locked
  twistLimitDegrees=45
  twistMotion=Locked
 }
 {bone1Name=dyn_coat_frnt_r, bone2Name=dyn_coat_MSTR_frnt_r, swing1LimitDegrees=30, swing2LimitDegrees=10, twistLimitDegrees=5}
 {bone1Name=dyn_coat_frnt_01_r, bone2Name=dyn_coat_frnt_r, swing1LimitDegrees=45, swing2LimitDegrees=10, twistLimitDegrees=5}
 {bone1Name=dyn_coat_frnt_l, bone2Name=dyn_coat_MSTR_frnt_l, swing1LimitDegrees=45, swing2LimitDegrees=25, twistLimitDegrees=10}
 {bone1Name=dyn_coat_frnt_01_l, bone2Name=dyn_coat_frnt_l, swing1LimitDegrees=45, swing2LimitDegrees=25, twistLimitDegrees=10}
 {bone1Name=dyn_coat_back_l, bone2Name=dyn_coat_MSTR_back_l, swing1LimitDegrees=10, swing2LimitDegrees=10, twistLimitDegrees=5}
 {bone1Name=dyn_coat_back_01_l, bone2Name=dyn_coat_back_l, swing1LimitDegrees=10, swing2LimitDegrees=5, twistLimitDegrees=2}
 {bone1Name=dyn_coat_back_r, bone2Name=dyn_coat_MSTR_back_r, swing1LimitDegrees=10, swing2LimitDegrees=10, twistLimitDegrees=5}
 {bone1Name=dyn_coat_back_01_r, bone2Name=dyn_coat_back_r, swing1LimitDegrees=10, swing2LimitDegrees=5, twistLimitDegrees=2}
 {bone1Name=dyn_coat_side_r, bone2Name=dyn_coat_MSTR_side_r, swing1LimitDegrees=10, swing2LimitDegrees=5, twistLimitDegrees=2}
 {bone1Name=dyn_coat_side_01_r, bone2Name=dyn_coat_side_r, swing1LimitDegrees=10, swing2LimitDegrees=5, twistLimitDegrees=2}
 {bone1Name=dyn_coat_side_l, bone2Name=dyn_coat_MSTR_side_l, swing1LimitDegrees=10, swing2LimitDegrees=5, twistLimitDegrees=2}
 {bone1Name=dyn_coat_side_01_l, bone2Name=dyn_coat_side_l, swing1LimitDegrees=10, swing2LimitDegrees=5, twistLimitDegrees=2}
 {
  bone1Name=dyn_coat_frnt_01_r
  bone2Name=dyn_coat_side_01_r
  swing1LimitDegrees=45
  swing1Motion=Locked
  swing2LimitDegrees=45
  swing2Motion=Locked
  twistLimitDegrees=45
  twistMotion=Locked
 }
 {
  bone1Name=dyn_coat_frnt_01_l
  bone2Name=dyn_coat_side_01_l
  swing1LimitDegrees=45
  swing1Motion=Locked
  swing2LimitDegrees=45
  swing2Motion=Locked
  twistLimitDegrees=45
  twistMotion=Locked
 }
 {
  bone1Name=dyn_MSTR_side_frt_R
  bone2Name=pelvis
  swing1LimitDegrees=45
  swing1Motion=Locked
  swing2LimitDegrees=45
  swing2Motion=Locked
  twistLimitDegrees=45
  twistMotion=Locked
 }
 {bone1Name=dyn_side_frt_R, bone2Name=dyn_MSTR_side_frt_R, swing1LimitDegrees=15, swing2LimitDegrees=5, twistLimitDegrees=0}
 {bone1Name=dyn_side_frt_R_01, bone2Name=dyn_side_frt_R, swing1LimitDegrees=10, swing2LimitDegrees=5, twistLimitDegrees=0}
 {
  bone1Name=dyn_MSTR_frnt_side_l
  bone2Name=pelvis
  swing1LimitDegrees=45
  swing1Motion=Locked
  swing2LimitDegrees=45
  swing2Motion=Locked
  twistLimitDegrees=45
  twistMotion=Locked
 }
 {bone1Name=dyn_frn_sidet_l, bone2Name=dyn_MSTR_frnt_side_l, swing1LimitDegrees=15, swing2LimitDegrees=5, twistLimitDegrees=0}
 {bone1Name=dyn_frnt_side_01_l, bone2Name=dyn_frn_sidet_l, swing1LimitDegrees=45, swing2LimitDegrees=20, twistLimitDegrees=0}
]
