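# Physics asset for the Henchman Bad medium male body, sky_drive variant, simulating the dangling coat.
# Pelvis and locked thigh capsules form the body volume, with coat chains hanging from pelvis through master bones into back, front and side segments, plus an extra right side-front piece.
# Back segments use capsules, while front and side panels use thin boxes.
# Narrow swing limits and mostly locked twist keep the cloth-like motion stiff.
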PhysicsAsset /Game/Characters/Player/Male/Medium/Bodies/M_MED_Henchman_Bad/Meshes/M_MED_Henchman_Bad2_Physics_sky_drive
bodies {
 dyn_MSTR_side_frt_R [{shapeType=Capsule, length=0.505, radius=0.505, shapeName=dyn_MSTR_side_frt_R_capsule, rotation=<rot 0 -90 0>}]
 dyn_coat_MSTR_back_l [{shapeType=Capsule, length=15.9, radius=0.5, shapeName=dyn_coat_MSTR_back_l_capsule, center=<0 -2.6129 0>}]
 dyn_coat_MSTR_back_r [{shapeType=Capsule, length=15.9, radius=0.5, shapeName=dyn_coat_MSTR_back_r_capsule, center=<0 -2.6129 0>}]
 dyn_coat_MSTR_frnt_l [{shapeType=Capsule, length=0, radius=0.5}]
 dyn_coat_MSTR_frnt_r [{shapeType=Capsule, length=0.505, radius=0.505, shapeName=dyn_coat_MSTR_frnt_r_capsule, rotation=<rot 0 -90 0>}]
 dyn_coat_MSTR_side_l [{shapeType=Capsule, length=0, radius=0.5}]
 dyn_coat_MSTR_side_r [{shapeType=Capsule, length=0.505, radius=0.505, shapeName=dyn_coat_MSTR_side_r_capsule, rotation=<rot 0 -90 0>}]
 dyn_coat_back_01_l [{shapeType=Capsule, length=12.8, radius=4.3, shapeName=dyn_coat_back_01_l_capsule, center=<0 -1.8013 0>}]
 dyn_coat_back_01_r [{shapeType=Capsule, length=13, radius=4.1, shapeName=dyn_coat_back_01_r_capsule, center=<0 -1.8013 0>}]
 dyn_coat_back_l [{shapeType=Capsule, length=13.3, radius=3.5, shapeName=dyn_coat_back_l_capsule, center=<0 -2.6129 0>}]
 dyn_coat_back_r [{shapeType=Capsule, length=13.3, radius=3.5, shapeName=dyn_coat_back_r_capsule, center=<0 -2.6129 0>}]
 dyn_coat_frnt_01_l [{shapeType=Box, extentX=7.5, extentY=1.5, extentZ=16.5, shapeName=dyn_coat_frnt_01_l_box, center=<4.5005 0 0>}]
 dyn_coat_frnt_01_r [
  {shapeType=Capsule, length=0, radius=0, shapeName=dyn_coat_frnt_01_r_capsule}
  {shapeType=Box, extentX=9.5, extentY=1.5, extentZ=16.5, shapeName=dyn_coat_frnt_01_r_box, center=<5.3394 0 0>}
 ]
 dyn_coat_frnt_l [{shapeType=Box, extentX=7.7, extentY=1.5, extentZ=16.5, shapeName=dyn_coat_frnt_l_box, center=<5.393 0 0>}]
 dyn_coat_frnt_r [
  {shapeType=Capsule, length=0, radius=0, shapeName=dyn_coat_frnt_r_capsule}
  {shapeType=Box, extentX=7.9, extentY=1.5, extentZ=12.5, shapeName=dyn_coat_frnt_r_box, center=<5.1915 0 0>}
 ]
 dyn_coat_side_01_l [{shapeType=Box, extentX=7.5, extentY=1.5, extentZ=12.5, shapeName=dyn_coat_side_01_l_box, center=<6.1906 0 0>}]
 dyn_coat_side_01_r [
  {shapeType=Capsule, length=0, radius=0, shapeName=dyn_coat_side_01_r_capsule}
  {shapeType=Box, extentX=7.5, extentY=1.5, extentZ=12.5, shapeName=dyn_coat_side_01_r_box, center=<5.3394 0 0>}
 ]
 dyn_coat_side_l [{shapeType=Box, extentX=7.5, extentY=1.5, extentZ=10.5, shapeName=dyn_coat_side_l_box, center=<6.1906 0 0>}]
 dyn_coat_side_r [
  {shapeType=Capsule, length=0, radius=0, shapeName=dyn_coat_side_r_capsule}
  {shapeType=Box, extentX=7.5, extentY=1.5, extentZ=12.5, shapeName=dyn_coat_side_r_box, center=<5.3394 0 0>}
 ]
 dyn_side_frt_R [{
   shapeType=Capsule
   length=6.4761
   radius=2.1382
   shapeName=dyn_side_frt_R_capsule
   center=<11.58 1.6146 3.5883>
   rotation=<rot 0 99.7383 -58.5382>
  }]
 pelvis [{shapeType=Capsule, length=9.8495, radius=9.8243, center=<-0.0062 -0.8187 -0.0691>}]
 thigh_l [{shapeType=Capsule, length=15.9082, radius=10.6708, center=<-10.0262 -0.4318 1.5531>, rotation=<rot 0 89.9999 -88.5886>}]
 thigh_r [{shapeType=Capsule, length=15.8082, radius=10.7708, center=<10.0262 0.4318 -1.5531>, rotation=<rot 0 87.9613 -89.9999>}]
}
constraints [
 {
  bone1Name=thigh_l
  bone2Name=pelvis
  swing1LimitDegrees=45
  swing1Motion=Locked
  swing2LimitDegrees=45
  swing2Motion=Locked
  twistLimitDegrees=45
  twistMotion=Locked
 }
 {
  bone1Name=thigh_r
  bone2Name=pelvis
  swing1LimitDegrees=45
  swing1Motion=Locked
  swing2LimitDegrees=45
  swing2Motion=Locked
  twistLimitDegrees=45
  twistMotion=Locked
 }
 {bone1Name=dyn_coat_MSTR_back_l, bone2Name=pelvis, swing1LimitDegrees=10, swing2LimitDegrees=5, twistLimitDegrees=1}
 {bone1Name=dyn_coat_back_l, bone2Name=dyn_coat_MSTR_back_l, swing1LimitDegrees=10, swing2LimitDegrees=5, twistLimitDegrees=1}
 {bone1Name=dyn_coat_back_01_l, bone2Name=dyn_coat_back_l, swing1LimitDegrees=7, swing2LimitDegrees=6, twistLimitDegrees=1}
 {bone1Name=dyn_coat_MSTR_back_r, bone2Name=pelvis, swing1LimitDegrees=10, swing2LimitDegrees=5, twistLimitDegrees=1}
 {bone1Name=dyn_coat_back_r, bone2Name=dyn_coat_MSTR_back_r, swing1LimitDegrees=10, swing2LimitDegrees=5, twistLimitDegrees=1}
 {bone1Name=dyn_coat_back_01_r, bone2Name=dyn_coat_back_r, swing1LimitDegrees=7, swing2LimitDegrees=6, twistLimitDegrees=1}
 {
  bone1Name=dyn_coat_MSTR_frnt_l
  bone2Name=pelvis
  swing1LimitDegrees=45
  swing1Motion=Locked
  swing2LimitDegrees=10
  swing2Motion=Locked
  twistLimitDegrees=45
  twistMotion=Locked
 }
 {
  bone1Name=dyn_coat_frnt_l
  bone2Name=dyn_coat_MSTR_frnt_l
  swing1LimitDegrees=7
  swing2LimitDegrees=1
  twistLimitDegrees=45
  twistMotion=Locked
 }
 {
  bone1Name=dyn_coat_frnt_01_l
  bone2Name=dyn_coat_frnt_l
  swing1LimitDegrees=10
  swing2LimitDegrees=1
  twistLimitDegrees=45
  twistMotion=Locked
 }
 {bone1Name=dyn_coat_MSTR_side_l, bone2Name=pelvis, swing1LimitDegrees=45, swing2LimitDegrees=10, twistLimitDegrees=45, twistMotion=Locked}
 {
  bone1Name=dyn_coat_side_l
  bone2Name=dyn_coat_MSTR_side_l
  swing1LimitDegrees=15
  swing2LimitDegrees=5
  twistLimitDegrees=45
  twistMotion=Locked
 }
 {
  bone1Name=dyn_coat_side_01_l
  bone2Name=dyn_coat_side_l
  swing1LimitDegrees=15
  swing2LimitDegrees=5
  twistLimitDegrees=45
  twistMotion=Locked
 }
 {
  bone1Name=dyn_coat_MSTR_frnt_r
  bone2Name=pelvis
  swing1LimitDegrees=45
  swing1Motion=Locked
  swing2LimitDegrees=10
  swing2Motion=Locked
  twistLimitDegrees=45
  twistMotion=Locked
 }
 {
  bone1Name=dyn_coat_frnt_r
  bone2Name=dyn_coat_MSTR_frnt_r
  swing1LimitDegrees=10
  swing2LimitDegrees=1
  twistLimitDegrees=45
  twistMotion=Locked
 }
 {
  bone1Name=dyn_coat_frnt_01_r
  bone2Name=dyn_coat_frnt_r
  swing1LimitDegrees=10
  swing2LimitDegrees=1
  swing2Motion=Locked
  twistLimitDegrees=45
  twistMotion=Locked
 }
 {
  bone1Name=dyn_coat_MSTR_side_r
  bone2Name=pelvis
  swing1LimitDegrees=45
  swing1Motion=Locked
  swing2LimitDegrees=45
  swing2Motion=Locked
  twistLimitDegrees=45
  twistMotion=Locked
 }
 {
  bone1Name=dyn_coat_side_r
  bone2Name=dyn_coat_MSTR_side_r
  swing1LimitDegrees=15
  swing2LimitDegrees=5
  twistLimitDegrees=45
  twistMotion=Locked
 }
 {
  bone1Name=dyn_coat_side_01_r
  bone2Name=dyn_coat_side_r
  swing1LimitDegrees=15
  swing2LimitDegrees=5
  twistLimitDegrees=45
  twistMotion=Locked
 }
 {
  bone1Name=dyn_MSTR_side_frt_R
  bone2Name=pelvis
  swing1LimitDegrees=45
  swing1Motion=Locked
  swing2LimitDegrees=45
  swing2Motion=Locked
  twistLimitDegrees=45
  twistMotion=Locked
 }
 {
  bone1Name=dyn_side_frt_R
  bone2Name=dyn_MSTR_side_frt_R
  swing1LimitDegrees=2
  swing2LimitDegrees=1
  twistLimitDegrees=45
  twistMotion=Locked
 }
]
