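# Control Rig for the male medium Henchman_Bad body that corrects coat, glute and knee dyn bones, apparently for crouch poses.
# Spherical pose readers on thighs and calves first set glute and knee transforms, then a sequence offsets right, left and front coat bones.
# Thigh angles and pelvis height refine some weights, with Kalman smoothing on the left.
# Some nodes are unwired, and one targets dyn_coat_MSTR_back_MID_, absent from the hierarchy.
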ControlRigBlueprint /Game/Characters/Player/Male/Medium/Bodies/M_MED_Henchman_Bad/Meshes/M_MED_Henchman_Bad_crouch_CtrlRig
graphs [
 RigVMFunctionLibrary {links [], nodes []}
 RigVMFunction_Sequence_1_ContainedGraph {nodes [
   n1 = RigVMFunctionEntryNode()
   RigVMFunction_Sequence = &RigVMFunction_Sequence(ExecuteContext=@n1.ExecuteContext)
   n2 = &RigVMFunction_Sequence(ExecuteContext=@RigVMFunction_Sequence.B)
   RigVMFunctionReturnNode(A=@RigVMFunction_Sequence.A, B=@n2.A, C=@n2.B)
  ]}
 RigVMModel {nodes [
   n1 = &RigVMFunction_MathDoubleMul(A=1, B=&RigVMFunction_MathDoubleRemap(SourceMaximum=20, SourceMinimum=0, TargetMaximum=1, TargetMinimum=0, Value=0).Result)
   &RigUnit_ModifyTransforms(
    ExecutePin=()
    ItemToModify=((Item=(Type=Bone,Name="dyn_coat_MSTR_frnt_r"),Transform=(Rotation=(X=0,Y=0,Z=0,W=1),Translation=(X=0,Y=0,Z=0),Scale3D=(X=1,Y=1,Z=1))))
    Weight=@n1.Result
    ItemToModify.0.Transform=/Script/RigVM.RigVMFunction_MathTransformFromSRT(Location=(X=0,Y=0,Z=0), Rotation=(X=0,Y=0,Z=10), Scale=(X=1,Y=1,Z=1)) {is=struct RigVMUnitNode}.Transform
   )
   > &RigUnit_ModifyTransforms(
    ItemToModify=((Item=(Type=Bone,Name="dyn_MSTR_side_frt_R"),Transform=(Rotation=(X=0,Y=0,Z=0,W=1),Translation=(X=0,Y=0,Z=0),Scale3D=(X=1,Y=1,Z=1))))
    WeightMinimum=0
    Weight=@n1.Result
    ItemToModify.0.Transform=&RigVMFunction_MathTransformFromSRT(Location=(X=0,Y=0,Z=0), Rotation=(X=0,Y=0,Z=45), Scale=(X=1,Y=1,Z=1)).Transform
   )
   > &RigUnit_ModifyTransforms(
    ItemToModify=((Item=(Type=Bone,Name="dyn_coat_MSTR_side_r"),Transform=(Rotation=(X=0,Y=0,Z=0,W=1),Translation=(X=0,Y=0,Z=0),Scale3D=(X=1,Y=1,Z=1))))
    WeightMinimum=0
    Weight=@n1.Result
    ItemToModify.0.Transform=&RigVMFunction_MathTransformFromSRT(Location=(X=0,Y=0,Z=0), Rotation=(X=0,Y=0,Z=30), Scale=(X=1,Y=1,Z=1)).Transform
   )
   n2 = &RigUnit_SphericalPoseReader(
    ActiveRegionScaleFactors=(PositiveWidth=4,NegativeWidth=0.1,PositiveHeight=3,NegativeHeight=4)
    ActiveRegionSize=0.1
    Debug=(bDrawDebug=false,bDraw2D=false,bDrawLocalAxes=false,DebugScale=30,DebugSegments=30,DebugThickness=0.2)
    DriverAxis=(X=-1,Y=0,Z=0)
    DriverItem=(Type=Bone,Name="thigh_l")
    FalloffRegionScaleFactors=(PositiveWidth=1,NegativeWidth=1.6,PositiveHeight=1,NegativeHeight=1)
    FalloffSize=0.4
    FlipWidthScaling=false
    OptionalParentItem=(Type=Bone,Name="pelvis")
    RotationOffset=(X=110,Y=15,Z=-5)
    ExecutePin=@RigVMFunction_Sequence.B
   )
   &RigUnit_ModifyTransforms(
    ExecutePin=()
    ItemToModify=((Item=(Type=Bone,Name="dyn_coat_MSTR_frnt_l"),Transform=(Rotation=(X=0,Y=0,Z=0,W=1),Translation=(X=0,Y=0,Z=0),Scale3D=(X=1,Y=1,Z=1))))
    WeightMinimum=0
    ItemToModify.0.Transform=&RigVMFunction_MathTransformFromSRT(Location=(X=0,Y=0,Z=0), Rotation=(X=0,Y=0,Z=30), RotationOrder=XYZ, Scale=(X=1,Y=1,Z=1)).Transform
    Weight=&RigVMFunction_MathDoubleRemap(SourceMaximum=1, SourceMinimum=0, TargetMaximum=0, TargetMinimum=1, Value=@n2.OutputParam).Result
   )
   RigUnit_GetRelativeTransformForItem = &RigUnit_GetRelativeTransformForItem(Child=(Type=Bone,Name="thigh_l"), Parent=(Type=Bone,Name="pelvis"), bChildInitial=False, bParentInitial=False)
   RigVMFunction_MathQuaternionToEuler = &RigVMFunction_MathQuaternionToEuler(RotationOrder=ZYX, Value=@RigUnit_GetRelativeTransformForItem.RelativeTransform.Rotation)
   Multiply = &RigVMFunction_MathDoubleMul(
    A=&RigVMFunction_MathDoubleRemap(SourceMaximum=0, SourceMinimum=-10, TargetMaximum=1, TargetMinimum=0, Value=@RigVMFunction_MathQuaternionToEuler.Result.X).Result
    B=&RigVMFunction_MathDoubleRemap(
     SourceMaximum=70
     SourceMinimum=40
     TargetMaximum=1
     TargetMinimum=0
     Value=&RigUnit_GetRelativeTransformForItem(Child=(Type=Bone,Name="pelvis"), Parent=(Type=Bone,Name="root"), bChildInitial=False, bParentInitial=False).RelativeTransform.Translation.Z
    ).Result
   )
   > &RigUnit_ModifyTransforms(
    ItemToModify=((Item=(Type=Bone,Name="dyn_coat_MSTR_frnt_l"),Transform=(Rotation=(X=0,Y=0,Z=0,W=1),Translation=(X=0,Y=0,Z=0),Scale3D=(X=1,Y=1,Z=1))))
    WeightMinimum=0
    Weight=@Multiply.Result
    ItemToModify.0.Transform=&RigVMFunction_MathTransformFromSRT(Location=(X=0,Y=0,Z=0), Rotation=(X=0,Y=0,Z=-25), RotationOrder=XYZ, Scale=(X=1,Y=1,Z=1)).Transform
   )
   RigUnit_BeginExecution = &RigUnit_BeginExecution()
   n3 = &RigUnit_SphericalPoseReader(
    ActiveRegionScaleFactors=(PositiveWidth=1.5,NegativeWidth=1,PositiveHeight=1,NegativeHeight=1)
    ActiveRegionSize=0.05
    Debug=(bDrawDebug=false,bDraw2D=false,bDrawLocalAxes=false,DebugScale=30,DebugSegments=30,DebugThickness=0.2)
    DriverAxis=(X=1,Y=0,Z=0)
    DriverItem=(Type=Bone,Name="thigh_l")
    FalloffRegionScaleFactors=(PositiveWidth=3,NegativeWidth=1,PositiveHeight=1,NegativeHeight=1)
    FalloffSize=0.15
    OptionalParentItem=(Type=Bone,Name="pelvis")
    RotationOffset=(X=-66,Y=0,Z=-25)
    ExecutePin=@RigUnit_BeginExecution.ExecutePin
   )
   RigVMFunction_MathTransformMake = /Script/RigVM.RigVMFunction_MathTransformMake(Rotation=(X=0,Y=0,Z=0,W=1), Scale=(X=1,Y=1,Z=1), Translation=(X=10,Y=-10,Z=28)) {is=struct RigVMUnitNode}
   > &RigUnit_SetTransform(
    Item=(Type=Bone,Name="dyn_glute")
    bInitial=False
    bPropagateToChildren=true
    Value=@RigVMFunction_MathTransformMake.Result
    Weight=@n3.OutputParam
   )
   > n4 = &RigUnit_SphericalPoseReader(
    ActiveRegionScaleFactors=(PositiveWidth=1.5,NegativeWidth=1.5,PositiveHeight=1,NegativeHeight=1)
    ActiveRegionSize=0.0051
    Debug=(bDrawDebug=false,bDraw2D=false,bDrawLocalAxes=false,DebugScale=30,DebugSegments=30,DebugThickness=0.2)
    DriverAxis=(X=1,Y=0,Z=0)
    DriverItem=(Type=Bone,Name="thigh_l")
    FalloffRegionScaleFactors=(PositiveWidth=3,NegativeWidth=1,PositiveHeight=1,NegativeHeight=1)
    FalloffSize=0.02
    OptionalParentItem=(Type=Bone,Name="pelvis")
    RotationOffset=(X=-65.5,Y=0,Z=-12.5)
   )
   > &RigUnit_SetTransform(
    Item=(Type=Bone,Name="dyn_coat_MSTR_back_MID_")
    Value=(Rotation=(X=-0.1209,Y=0.675,Z=-0.1209,W=0.7177),Translation=(X=10,Y=-22.6656,Z=47.0635),Scale3D=(X=1,Y=1,Z=1))
    bInitial=False
    Weight=@n4.OutputParam
   )
   > n5 = &RigUnit_SphericalPoseReader(
    ActiveRegionScaleFactors=(PositiveWidth=1.5,NegativeWidth=1,PositiveHeight=1,NegativeHeight=1)
    ActiveRegionSize=0.05
    Debug=(bDrawDebug=false,bDraw2D=false,bDrawLocalAxes=false,DebugScale=30,DebugSegments=30,DebugThickness=0.2)
    DriverAxis=(X=-1,Y=0,Z=0)
    DriverItem=(Type=Bone,Name="calf_l")
    FalloffRegionScaleFactors=(PositiveWidth=3,NegativeWidth=1,PositiveHeight=1,NegativeHeight=1)
    FalloffSize=0.15
    OptionalParentItem=(Type=Bone,Name="pelvis")
    RotationOffset=(X=-80,Y=180,Z=45)
   )
   > RigUnit_SetTransform = &RigUnit_SetTransform(
    Item=(Type=Bone,Name="dyn_knee_L")
    Value=(Rotation=(X=0.8364,Y=0.0997,Z=-0.5368,W=0.0488),Translation=(X=12.2192,Y=27.9829,Z=10.3894),Scale3D=(X=0.5,Y=1,Z=0.009))
    bInitial=False
    Weight=@n5.OutputParam
   )
   > n6 = &RigUnit_SphericalPoseReader(
    ActiveRegionScaleFactors=(PositiveWidth=1.5,NegativeWidth=1,PositiveHeight=1,NegativeHeight=1)
    ActiveRegionSize=0.05
    Debug=(bDrawDebug=false,bDraw2D=false,bDrawLocalAxes=false,DebugScale=30,DebugSegments=30,DebugThickness=0.2)
    DriverAxis=(X=1,Y=0,Z=0)
    DriverItem=(Type=Bone,Name="calf_r")
    FalloffRegionScaleFactors=(PositiveWidth=3,NegativeWidth=1,PositiveHeight=1,NegativeHeight=1)
    FalloffSize=0.15
    OptionalParentItem=(Type=Bone,Name="pelvis")
    RotationOffset=(X=125,Y=0,Z=55)
   )
   > &RigUnit_SetTransform(
    Item=(Type=Bone,Name="dyn_Knee_R")
    Value=(Rotation=(X=-0.2627,Y=-0.5321,Z=0.0862,W=0.8003),Translation=(X=-26.5406,Y=15.4801,Z=7.7886),Scale3D=(X=1,Y=1,Z=0.1))
    bInitial=False
    Weight=@n6.OutputParam
   )
   > RigVMFunction_Sequence = RigVMAggregateNode()
   n7 = &RigUnit_SphericalPoseReader(
    ActiveRegionScaleFactors=(PositiveWidth=1,NegativeWidth=0.2,PositiveHeight=1,NegativeHeight=1)
    ActiveRegionSize=0.3
    Debug=(bDrawDebug=false,bDraw2D=false,bDrawLocalAxes=false,DebugScale=30,DebugSegments=30,DebugThickness=0.2)
    DriverAxis=(X=1,Y=0,Z=0)
    DriverItem=(Type=Bone,Name="thigh_r")
    FalloffRegionScaleFactors=(PositiveWidth=1,NegativeWidth=1.5,PositiveHeight=1,NegativeHeight=1)
    FalloffSize=0.4
    OptionalParentItem=(Type=Bone,Name="pelvis")
    RotationOffset=(X=-110,Y=180,Z=5)
    ExecutePin=@RigVMFunction_Sequence.A
   )
   > &RigUnit_ModifyTransforms(
    ItemToModify=((Item=(Type=Bone,Name="dyn_coat_MSTR_frnt_r"),Transform=(Rotation=(X=0,Y=0,Z=0,W=1),Translation=(X=0,Y=0,Z=0),Scale3D=(X=1,Y=1,Z=1))))
    Weight=&RigVMFunction_AnimEvalRichCurve(Curve=(EditorCurveData=(Keys=((),(Time=0.5,Value=1)))), SourceMinimum=0, TargetMinimum=0, Value=@n7.OutputParam).Result
    ItemToModify.0.Transform=&RigVMFunction_MathTransformFromSRT(Location=(X=0,Y=0,Z=0), Rotation=(X=0,Y=0,Z=-20), Scale=(X=1,Y=1,Z=1)).Transform
   )
   > &RigUnit_ModifyTransforms(
    ItemToModify=((Item=(Type=Bone,Name="dyn_coat_MSTR_frnt_r"),Transform=(Rotation=(X=0,Y=0,Z=0,W=1),Translation=(X=0,Y=0,Z=0),Scale3D=(X=1,Y=1,Z=1))))
    Weight=@n7.OutputParam
    ItemToModify.0.Transform=&RigVMFunction_MathTransformFromSRT(Location=(X=-1.5,Y=1,Z=1), Rotation=(X=-10,Y=-5,Z=-87), Scale=(X=1,Y=1,Z=1)).Transform
   )
   > &RigUnit_ModifyTransforms(
    ItemToModify=((Item=(Type=Bone,Name="dyn_MSTR_side_frt_R"),Transform=(Rotation=(X=0,Y=0,Z=0,W=1),Translation=(X=0,Y=0,Z=0),Scale3D=(X=1,Y=1,Z=1))))
    Weight=@n7.OutputParam
    ItemToModify.0.Transform=&RigVMFunction_MathTransformFromSRT(Location=(X=0,Y=0,Z=0), Rotation=(X=30,Y=0,Z=-90), Scale=(X=1,Y=1,Z=1)).Transform
   )
   > &RigUnit_ModifyTransforms(
    ItemToModify=((Item=(Type=Bone,Name="dyn_coat_MSTR_side_r"),Transform=(Rotation=(X=0,Y=0,Z=0,W=1),Translation=(X=0,Y=0,Z=0),Scale3D=(X=1,Y=1,Z=1))))
    Weight=@n7.OutputParam
    ItemToModify.0.Transform=&RigVMFunction_MathTransformFromSRT(Location=(X=-3,Y=0,Z=0), Rotation=(X=-30,Y=30,Z=-50), Scale=(X=1,Y=1,Z=1)).Transform
   )
   n8 = &RigUnit_GetTransform(Item=(Type=Bone,Name="thigh_r"), Space=LocalSpace, bInitial=False)
   n9 = &RigVMFunction_MathDoubleMul(
    B=&RigVMFunction_MathDoubleRemap(
     SourceMaximum=-45
     SourceMinimum=10
     TargetMaximum=1
     TargetMinimum=0
     Value=&RigVMFunction_MathQuaternionToEuler(RotationOrder=ZYX, Value=@n8.Transform.Rotation).Result.X
    ).Result
    A=@n7.OutputParam
   )
   > &RigUnit_ModifyTransforms(
    ItemToModify=((Item=(Type=Bone,Name="dyn_coat_MSTR_frnt_r"),Transform=(Rotation=(X=0,Y=0,Z=0,W=1),Translation=(X=0,Y=0,Z=0),Scale3D=(X=1,Y=1,Z=1))))
    Weight=@n9.Result
    ItemToModify.0.Transform=&RigVMFunction_MathTransformFromSRT(Location=(X=-2,Y=4.5,Z=3), Rotation=(X=0,Y=50,Z=25), Scale=(X=1,Y=1,Z=1)).Transform
   )
   > &RigUnit_ModifyTransforms(
    ItemToModify=((Item=(Type=Bone,Name="dyn_MSTR_side_frt_R"),Transform=(Rotation=(X=0,Y=0,Z=0,W=1),Translation=(X=0,Y=0,Z=0),Scale3D=(X=1,Y=1,Z=1))))
    WeightMinimum=0
    Weight=@n9.Result
    ItemToModify.0.Transform=&RigVMFunction_MathTransformFromSRT(Location=(X=0,Y=0,Z=0), Rotation=(X=0,Y=40,Z=65), Scale=(X=1,Y=1,Z=1)).Transform
   )
   > &RigUnit_ModifyTransforms(
    ItemToModify=((Item=(Type=Bone,Name="dyn_coat_MSTR_side_r"),Transform=(Rotation=(X=0,Y=0,Z=0,W=1),Translation=(X=0,Y=0,Z=0),Scale3D=(X=1,Y=1,Z=1))))
    WeightMinimum=0
    Weight=@n9.Result
    ItemToModify.0.Transform=&RigVMFunction_MathTransformFromSRT(Location=(X=0,Y=0,Z=0), Rotation=(X=0,Y=0,Z=45), Scale=(X=1,Y=1,Z=1)).Transform
   )
   n10 = &RigVMFunction_MathDoubleRemap(SourceMaximum=1, SourceMinimum=0, TargetMaximum=0, TargetMinimum=1, Value=@n7.OutputParam)
   n11 = &RigUnit_GetRelativeTransformForItem(Child=(Type=Bone,Name="pelvis"), Parent=(Type=Bone,Name="root"), bChildInitial=False, bParentInitial=False)
   n12 = &RigVMFunction_MathDoubleRemap(SourceMaximum=80, SourceMinimum=60, TargetMaximum=1, TargetMinimum=0, Value=@n11.RelativeTransform.Translation.Z)
   n13 = &RigVMFunction_MathDoubleMul(A=@n10.Result, B=@n12.Result)
   > &RigUnit_ModifyTransforms(
    ItemToModify=((Item=(Type=Bone,Name="dyn_coat_MSTR_frnt_r"),Transform=(Rotation=(X=0,Y=0,Z=0,W=1),Translation=(X=0,Y=0,Z=0),Scale3D=(X=1,Y=1,Z=1))))
    WeightMinimum=0
    Weight=@n13.Result
    ItemToModify.0.Transform=&RigVMFunction_MathTransformFromSRT(Location=(X=0,Y=0,Z=0), Rotation=(X=0,Y=0,Z=5), RotationOrder=XYZ, Scale=(X=1,Y=1,Z=1)).Transform
   )
   > &RigUnit_ModifyTransforms(
    ItemToModify=((Item=(Type=Bone,Name="dyn_coat_MSTR_side_r"),Transform=(Rotation=(X=0,Y=0,Z=0,W=1),Translation=(X=0,Y=0,Z=0),Scale3D=(X=1,Y=1,Z=1))))
    WeightMinimum=0
    Weight=@n13.Result
    ItemToModify.0.Transform=&RigVMFunction_MathTransformFromSRT(Location=(X=0,Y=0,Z=0), Rotation=(X=0,Y=0,Z=-5), RotationOrder=XYZ, Scale=(X=1,Y=1,Z=1)).Transform
   )
   > &RigUnit_ModifyTransforms(
    ItemToModify=((Item=(Type=Bone,Name="dyn_MSTR_side_frt_R"),Transform=(Rotation=(X=0,Y=0,Z=0,W=1),Translation=(X=0,Y=0,Z=0),Scale3D=(X=1,Y=1,Z=1))))
    WeightMinimum=0
    Weight=@n13.Result
    ItemToModify.0.Transform=&RigVMFunction_MathTransformFromSRT(Location=(X=0,Y=0,Z=0), Rotation=(X=0,Y=0,Z=10), RotationOrder=XYZ, Scale=(X=1,Y=1,Z=1)).Transform
   )
   > n14 = &RigUnit_SphericalPoseReader(
    ActiveRegionScaleFactors=(PositiveWidth=1,NegativeWidth=1,PositiveHeight=1,NegativeHeight=1)
    ActiveRegionSize=0.01
    Debug=(bDrawDebug=true,bDraw2D=false,bDrawLocalAxes=false,DebugScale=30,DebugSegments=30,DebugThickness=0.2)
    DriverAxis=(X=1,Y=0,Z=0)
    DriverItem=(Type=Bone,Name="thigh_r")
    FalloffRegionScaleFactors=(PositiveWidth=1,NegativeWidth=1,PositiveHeight=0,NegativeHeight=1)
    FalloffSize=0.1
    OptionalParentItem=(Type=Bone,Name="pelvis")
    RotationOffset=(X=-110,Y=160,Z=10)
   )
   n15 = &RigVMFunction_MathDoubleMul(B=@n12.Result, A=@n7.OutputParam)
   > &RigUnit_ModifyTransforms(
    ItemToModify=((Item=(Type=Bone,Name="dyn_coat_MSTR_frnt_r"),Transform=(Rotation=(X=0,Y=0,Z=0,W=1),Translation=(X=0,Y=0,Z=0),Scale3D=(X=1,Y=1,Z=1))))
    ItemToModify.0.Transform=&RigVMFunction_MathTransformFromSRT(Location=(X=0,Y=5,Z=5), Rotation=(X=0,Y=0,Z=30), Scale=(X=1,Y=1,Z=1)).Transform
    Weight=@n15.Result
   )
   &RigUnit_ModifyTransforms(
    ItemToModify=((Item=(Type=Bone,Name="dyn_coat_MSTR_frnt_l"),Transform=(Rotation=(X=-0.0151,Y=0.0858,Z=0.173,W=0.9811),Translation=(X=7,Y=0,Z=-3),Scale3D=(X=1,Y=1,Z=1))))
    WeightMinimum=0
    ExecutePin=@RigVMFunction_Sequence.C
    Weight=@n5.OutputParam
   )
   > &RigUnit_ModifyTransforms(
    ItemToModify=((Item=(Type=Bone,Name="dyn_coat_MSTR_frnt_r"),Transform=(Rotation=(X=0,Y=0,Z=0.2164,W=0.9763),Translation=(X=0,Y=-1,Z=0),Scale3D=(X=1,Y=1,Z=1))))
    WeightMinimum=0
    Weight=@n6.OutputParam
   )
   &RigVMFunction_KalmanFloat(BufferSize=1, Value=0) {template="KalmanFilter::Execute(in Value,in BufferSize,out Result)"}
   &RigUnit_ModifyTransforms(
    ExecutePin=()
    ItemToModify=((Item=(Type=Bone,Name="dyn_MSTR_side_frt_R"),Transform=(Rotation=(X=0,Y=0,Z=0,W=1),Translation=(X=0,Y=0,Z=0),Scale3D=(X=1,Y=1,Z=1))))
    ItemToModify.0.Transform=&RigVMFunction_MathTransformFromSRT(Location=(X=0,Y=0,Z=0), Rotation=(X=0,Y=0,Z=0), Scale=(X=1,Y=1,Z=1)).Transform
    Weight=@n15.Result
   )
   n16 = &RigVMFunction_KalmanFloat(
    BufferSize=7
    Value=&RigVMFunction_MathDoubleRemap(SourceMaximum=-36, SourceMinimum=-20, TargetMaximum=1, TargetMinimum=0, Value=@RigVMFunction_MathQuaternionToEuler.Result.Y).Result
   ) {template="KalmanFilter::Execute(in Value,in BufferSize,out Result)"}
   &RigUnit_ModifyTransforms(
    ItemToModify=((Item=(Type=Bone,Name="dyn_frn_sidet_l"),Transform=(Rotation=(X=0,Y=0,Z=0,W=1),Translation=(X=0,Y=0,Z=0),Scale3D=(X=1,Y=1,Z=1))))
    WeightMinimum=0
    ExecutePin=@n22.ExecutePin
    ItemToModify.0.Transform=&RigVMFunction_MathTransformFromSRT(Location=(X=0,Y=0,Z=0), Rotation=(X=-30,Y=20,Z=-70), RotationOrder=XYZ, Scale=(X=1,Y=1,Z=1)).Transform
    Weight=@n16.Result
   )
   n17 = &RigVMFunction_MathDoubleRemap(SourceMaximum=-160, SourceMinimum=-115, TargetMaximum=1, TargetMinimum=0, Value=@RigVMFunction_MathQuaternionToEuler.Result.Z)
   &RigUnit_ModifyTransforms(
    ExecutePin=()
    ItemToModify=((Item=(Type=Bone,Name="dyn_coat_MSTR_frnt_l"),Transform=(Rotation=(X=0,Y=0,Z=0,W=1),Translation=(X=0,Y=0,Z=0),Scale3D=(X=1,Y=1,Z=1))))
    WeightMinimum=0
    ItemToModify.0.Transform=&RigVMFunction_MathTransformFromSRT(Location=(X=-5,Y=-6,Z=-5), Rotation=(X=-20,Y=-10,Z=-10), RotationOrder=XYZ, Scale=(X=1,Y=1,Z=1)).Transform
    Weight=&RigVMFunction_MathDoubleMul(
     B=&RigVMFunction_KalmanFloat(BufferSize=4, Value=@n17.Result) {template="KalmanFilter::Execute(in Value,in BufferSize,out Result)"}.Result
     A=@n2.OutputParam
    ).Result
   )
   Power = /Script/RigVM.RigVMFunction_MathDoublePow(B=0.2, A=@n10.Result) {is=struct RigVMUnitNode, template="Power::Execute(in A,in B,out Result)"}
   &RigVMFunction_MathDoubleMul(B=@Power.Result, A=@n14.OutputParam)
   &RigVMFunction_MathDoubleRemap(SourceMaximum=80, SourceMinimum=50, TargetMaximum=1, TargetMinimum=0, Value=@n11.RelativeTransform.Translation.Z)
   n18 = &RigVMFunction_MathDoubleRemap(SourceMaximum=13, SourceMinimum=0, TargetMaximum=1, TargetMinimum=0, Value=@RigVMFunction_MathQuaternionToEuler.Result.X)
   RigVMRerouteNode(Value=@n18.Result)
   RigVMRerouteNode(Value=@n2.OutputParam)
   &RigVMFunction_MathTransformFromSRT(Location=(X=-1,Y=-2.2,Z=0), Rotation=(X=20,Y=-20,Z=0), RotationOrder=XYZ, Scale=(X=1,Y=1,Z=1))
   &RigVMFunction_MathTransformFromSRT(Location=(X=0,Y=0,Z=0), Rotation=(X=0,Y=0,Z=15), RotationOrder=XYZ, Scale=(X=1,Y=1,Z=1))
   &RigVMFunction_MathTransformFromSRT(Location=(X=0,Y=0,Z=0), Rotation=(X=30,Y=0,Z=60), Scale=(X=1,Y=1,Z=1))
   &RigVMFunction_MathTransformFromSRT(Location=(X=0,Y=0,Z=0), Rotation=(X=0,Y=0,Z=40), Scale=(X=1,Y=1,Z=1))
   n19 = &RigVMFunction_KalmanFloat(BufferSize=9, Value=&RigVMFunction_MathDoubleMul(B=@n18.Result, A=@n2.OutputParam).Result) {template="KalmanFilter::Execute(in Value,in BufferSize,out Result)"}
   Remap = &RigVMFunction_MathDoubleRemap(SourceMaximum=-120, SourceMinimum=-108, TargetMaximum=1, TargetMinimum=0, Value=@RigVMFunction_MathQuaternionToEuler.Result.Z)
   n20 = &RigVMFunction_KalmanFloat(BufferSize=4, Value=@Remap.Result) {template="KalmanFilter::Execute(in Value,in BufferSize,out Result)"}
   &RigUnit_ModifyTransforms(
    ItemToModify=((Item=(Type=Bone,Name="dyn_coat_MSTR_frnt_l"),Transform=(Rotation=(X=0,Y=0,Z=0,W=1),Translation=(X=0,Y=0,Z=0),Scale3D=(X=1,Y=1,Z=1))))
    WeightMinimum=0
    ExecutePin=@n23.ExecutePin
    ItemToModify.0.Transform=&RigVMFunction_MathTransformFromSRT(Location=(X=-5,Y=-4,Z=0), Rotation=(X=0,Y=10,Z=-70), RotationOrder=XYZ, Scale=(X=1,Y=1,Z=1)).Transform
    Weight=@n2.OutputParam
   )
   > &RigUnit_ModifyTransforms(
    ItemToModify=((Item=(Type=Bone,Name="dyn_MSTR_frnt_side_l"),Transform=(Rotation=(X=0,Y=0,Z=0,W=1),Translation=(X=0,Y=0,Z=0),Scale3D=(X=1,Y=1,Z=1))))
    WeightMinimum=0
    ItemToModify.0.Transform=&RigVMFunction_MathTransformFromSRT(Location=(X=-2,Y=0,Z=0), Rotation=(X=50,Y=0,Z=-70), RotationOrder=XYZ, Scale=(X=1,Y=1,Z=1)).Transform
    Weight=@n2.OutputParam
   )
   > &RigUnit_ModifyTransforms(
    ItemToModify=((Item=(Type=Bone,Name="dyn_coat_MSTR_side_l"),Transform=(Rotation=(X=0,Y=0,Z=0,W=1),Translation=(X=0,Y=0,Z=0),Scale3D=(X=1,Y=1,Z=1))))
    WeightMinimum=0
    ItemToModify.0.Transform=&RigVMFunction_MathTransformFromSRT(Location=(X=-2,Y=0,Z=0), Rotation=(X=30,Y=-35,Z=-55), RotationOrder=XYZ, Scale=(X=1,Y=1,Z=1)).Transform
    Weight=@n2.OutputParam
   )
   > &RigUnit_ModifyTransforms(
    ItemToModify=((Item=(Type=Bone,Name="dyn_coat_MSTR_back_l"),Transform=(Rotation=(X=0,Y=0,Z=0,W=1),Translation=(X=0,Y=0,Z=0),Scale3D=(X=1,Y=1,Z=1))))
    Weight=1
    WeightMinimum=0
    ItemToModify.0.Transform=&RigVMFunction_MathTransformFromSRT(Location=(X=0,Y=0,Z=3), Rotation=(X=0,Y=0,Z=-15), RotationOrder=XYZ, Scale=(X=1,Y=1,Z=1)).Transform
   )
   > &RigUnit_ModifyTransforms(
    ItemToModify=((Item=(Type=Bone,Name="dyn_coat_MSTR_frnt_l"),Transform=(Rotation=(X=0,Y=0,Z=0,W=1),Translation=(X=0,Y=0,Z=0),Scale3D=(X=1,Y=1,Z=1))))
    WeightMinimum=0
    ItemToModify.0.Transform=&RigVMFunction_MathTransformFromSRT(Location=(X=0,Y=0,Z=0), Rotation=(X=20,Y=-30,Z=0), RotationOrder=XYZ, Scale=(X=1,Y=1,Z=1)).Transform
    Weight=@n19.Result
   )
   > &RigUnit_ModifyTransforms(
    ItemToModify=((Item=(Type=Bone,Name="dyn_coat_MSTR_side_l"),Transform=(Rotation=(X=0,Y=0,Z=0,W=1),Translation=(X=0,Y=0,Z=0),Scale3D=(X=1,Y=1,Z=1))))
    WeightMinimum=0
    ItemToModify.0.Transform=&RigVMFunction_MathTransformFromSRT(Location=(X=0,Y=0,Z=0), Rotation=(X=0,Y=0,Z=30), RotationOrder=XYZ, Scale=(X=1,Y=1,Z=1)).Transform
    Weight=@n19.Result
   )
   > &RigUnit_ModifyTransforms(
    ItemToModify=((Item=(Type=Bone,Name="dyn_frn_sidet_l"),Transform=(Rotation=(X=0,Y=0,Z=0,W=1),Translation=(X=0,Y=0,Z=0),Scale3D=(X=1,Y=1,Z=1))))
    WeightMinimum=0
    ItemToModify.0.Transform=&RigVMFunction_MathTransformFromSRT(Location=(X=0,Y=0,Z=0), Rotation=(X=0,Y=0,Z=30), RotationOrder=XYZ, Scale=(X=1,Y=1,Z=1)).Transform
    Weight=@n19.Result
   )
   n21 = &RigVMFunction_MathDoubleMul(B=@n20.Result, A=@n2.OutputParam)
   > &RigUnit_ModifyTransforms(
    ItemToModify=((Item=(Type=Bone,Name="dyn_coat_MSTR_frnt_l"),Transform=(Rotation=(X=0,Y=0,Z=0,W=1),Translation=(X=0,Y=0,Z=0),Scale3D=(X=1,Y=1,Z=1))))
    WeightMinimum=0
    ItemToModify.0.Transform=&RigVMFunction_MathTransformFromSRT(Location=(X=0,Y=-5.5,Z=0), Rotation=(X=0,Y=20,Z=0), RotationOrder=XYZ, Scale=(X=1,Y=1,Z=1)).Transform
    Weight=@n21.Result
   )
   > &RigUnit_ModifyTransforms(
    ItemToModify=((Item=(Type=Bone,Name="dyn_coat_MSTR_side_l"),Transform=(Rotation=(X=0,Y=0,Z=0,W=1),Translation=(X=0,Y=0,Z=0),Scale3D=(X=1,Y=1,Z=1))))
    WeightMinimum=0
    ItemToModify.0.Transform=&RigVMFunction_MathTransformFromSRT(Location=(X=-1,Y=-2.2,Z=0), Rotation=(X=0,Y=0,Z=5), RotationOrder=XYZ, Scale=(X=1,Y=1,Z=1)).Transform
    Weight=@n21.Result
   )
   > &RigUnit_ModifyTransforms(
    ItemToModify=((Item=(Type=Bone,Name="dyn_MSTR_frnt_side_l"),Transform=(Rotation=(X=0,Y=0,Z=0,W=1),Translation=(X=0,Y=0,Z=0),Scale3D=(X=1,Y=1,Z=1))))
    WeightMinimum=0
    ItemToModify.0.Transform=&RigVMFunction_MathTransformFromSRT(Location=(X=0,Y=0,Z=4), Rotation=(X=0,Y=-20,Z=-20), RotationOrder=XYZ, Scale=(X=1,Y=1,Z=1)).Transform
    Weight=@n21.Result
   )
   > &RigUnit_ModifyTransforms(
    ItemToModify=((Item=(Type=Bone,Name="dyn_coat_MSTR_side_l"),Transform=(Rotation=(X=0,Y=0,Z=0,W=1),Translation=(X=0,Y=0,Z=0),Scale3D=(X=1,Y=1,Z=1))))
    WeightMinimum=0
    ItemToModify.0.Transform=&RigVMFunction_MathTransformFromSRT(Location=(X=0,Y=0,Z=0), Rotation=(X=0,Y=0,Z=-43), RotationOrder=XYZ, Scale=(X=1,Y=1,Z=1)).Transform
    Weight=@n16.Result
   )
   > n22 = &RigUnit_ModifyTransforms(
    ItemToModify=((Item=(Type=Bone,Name="dyn_coat_MSTR_frnt_l"),Transform=(Rotation=(X=0,Y=0,Z=0,W=1),Translation=(X=0,Y=0,Z=0),Scale3D=(X=1,Y=1,Z=1))))
    WeightMinimum=0
    ItemToModify.0.Transform=&RigVMFunction_MathTransformFromSRT(Location=(X=0,Y=0,Z=0), Rotation=(X=0,Y=25,Z=0), RotationOrder=XYZ, Scale=(X=1,Y=1,Z=1)).Transform
    Weight=@n16.Result
   )
   RigVMFunction_AnimEvalRichCurve = &RigVMFunction_AnimEvalRichCurve(Curve=(EditorCurveData=(Keys=((),(Time=0.5,Value=1)))), SourceMinimum=0, TargetMinimum=0, Value=@n2.OutputParam)
   n23 = &RigUnit_ModifyTransforms(
    ItemToModify=((Item=(Type=Bone,Name="dyn_coat_MSTR_frnt_l"),Transform=(Rotation=(X=0,Y=0,Z=0,W=1),Translation=(X=0,Y=0,Z=0),Scale3D=(X=1,Y=1,Z=1))))
    WeightMinimum=0
    Weight=@RigVMFunction_AnimEvalRichCurve.Result
    ItemToModify.0.Transform=&RigVMFunction_MathTransformFromSRT(Location=(X=0,Y=0,Z=0), Rotation=(X=0,Y=0,Z=-10), RotationOrder=XYZ, Scale=(X=1,Y=1,Z=1)).Transform
    ExecutePin=@n2.ExecutePin
   )
  ]}
]
hierarchy [
 root {loc=<0 0 0>, rot=<0 0 0 1>}
 attach {parent=root, loc=<0 0 0>, rot=<0 0 0 1>}
 pelvis {parent=root, loc=<0 -0.7071 93.1191>, rot=<0.0127 0.707 -0.0127 -0.707>}
 spine_01 {parent=pelvis, loc=<6.7494 -0.6388 0>, rot=<0 0 0.034 -0.9994>}
 spine_02 {parent=spine_01, loc=<7.2518 -0.035 0>, rot=<0 0 0.0238 -0.9997>}
 spine_03 {parent=spine_02, loc=<6.0104 0 0>, rot=<0 0 0.0149 -0.9999>}
 spine_04 {parent=spine_03, loc=<8.1024 0 0>, rot=<0 0 -0.0132 -0.9999>}
 spine_05 {parent=spine_04, loc=<10.0146 0 0>, rot=<0 0 -0.0015 -1>}
 clavicle_l {parent=spine_05, loc=<4.3575 -0.1466 -2.3223>, rot=<0.0465 -0.7289 -0.009 -0.6829>}
 upperarm_l {parent=clavicle_l, loc=<15.5111 0 0>, rot=<0.0226 -0.3234 -0.0088 -0.9459>}
 lowerarm_l {parent=upperarm_l, loc=<26.6514 0 0>, rot=<0.0391 -0.0099 -0.2453 -0.9686>}
 hand_l {parent=lowerarm_l, loc=<22.3491 0 0>, rot=<-0.7081 -0.0507 0.0975 -0.6975>}
 index_metacarpal_l {parent=hand_l, loc=<1.6535 0.0183 -2.96>, rot=<0.0158 -0.0124 -0.0173 -0.9996>}
 index_01_l {parent=index_metacarpal_l, loc=<8.3008 0.0357 -0.0524>, rot=<-0.0044 0.0485 0.0924 -0.9945>}
 index_02_l {parent=index_01_l, loc=<3.3587 -0.3004 0.0452>, rot=<-0.002 0.0027 0.3225 -0.9466>}
 index_03_l {parent=index_02_l, loc=<2.5578 -0.123 -0.0252>, rot=<-0.002 0.0067 0.2558 -0.9667>}
 middle_metacarpal_l {parent=hand_l, loc=<2.1856 -0.1725 -0.5984>, rot=<0.006 0.0168 -0.0329 -0.9993>}
 middle_01_l {parent=middle_metacarpal_l, loc=<7.3275 -0.2447 0.0129>, rot=<-0.0208 0.0042 0.0858 -0.9961>}
 middle_02_l {parent=middle_01_l, loc=<4.0123 0.1401 0.1175>, rot=<-0.0218 0.0258 0.3512 -0.9357>}
 middle_03_l {parent=middle_02_l, loc=<2.6077 0.0534 0.1373>, rot=<0.0082 0.0188 0.1708 -0.9851>}
 pinky_metacarpal_l {parent=hand_l, loc=<1.7616 -0.9753 3.3814>, rot=<-0.0397 0.0814 -0.069 -0.9935>}
 pinky_01_l {parent=pinky_metacarpal_l, loc=<7.1709 -0.4831 -0.1094>, rot=<-0.0654 -0.0035 0.2242 -0.9723>}
 pinky_02_l {parent=pinky_01_l, loc=<2.8274 0.0893 -0.1043>, rot=<0.0037 0.0011 0.2854 -0.9584>}
 pinky_03_l {parent=pinky_02_l, loc=<2.2046 0.0307 -0.0078>, rot=<-0.0002 0.0098 0.1445 -0.9895>}
 ring_metacarpal_l {parent=hand_l, loc=<2.2074 -0.3625 1.3638>, rot=<-0.0284 0.0722 -0.0458 -0.9959>}
 ring_01_l {parent=ring_metacarpal_l, loc=<7.0295 -0.6197 -0.1446>, rot=<-0.0489 -0.0252 0.1892 -0.9804>}
 ring_02_l {parent=ring_01_l, loc=<3.6174 0.4647 -0.1027>, rot=<0 0 0.2806 -0.9598>}
 ring_03_l {parent=ring_02_l, loc=<2.4152 0.114 0.0199>, rot=<0 0 0.1189 -0.9929>}
 thumb_01_l {parent=hand_l, loc=<1.3118 -0.3191 -2.459>, rot=<0.6546 -0.4042 0.0664 -0.6354>}
 thumb_02_l {parent=thumb_01_l, loc=<5.9336 -0.0448 0.0401>, rot=<-0.0037 0.0213 0.298 -0.9543>}
 thumb_03_l {parent=thumb_02_l, loc=<2.2223 0.3682 -0.0907>, rot=<0.0118 -0.013 0.0066 -0.9998>}
 weapon_l {parent=hand_l, loc=<-0.1974 -3.4111 0.3811>, rot=<0 0 0 1>}
 lowerarm_twist_01_l {parent=lowerarm_l, loc=<17.8549 0 0>, rot=<0 0 0 1>}
 lowerarm_twist_02_l {parent=lowerarm_l, loc=<9.6199 0 0>, rot=<0 0 0 1>}
 upperarm_twist_01_l {parent=upperarm_l, loc=<0.0041 0 0>, rot=<0 0 0 1>}
 deform_upperarm_bk_base_l {parent=upperarm_twist_01_l, loc=<-3.2582 -5.546 -2.0366>, rot=<0 -0.0007 -0.0191 -0.9998>}
 deform_upperarm_bk_l {parent=deform_upperarm_bk_base_l, loc=<0 0 0>, rot=<0 0 0 1>}
 deform_upperarm_fr_base_l {parent=upperarm_twist_01_l, loc=<-2.8462 3.9433 -2.0158>, rot=<0 -0.0007 -0.0191 -0.9998>}
 deform_upperarm_fr_l {parent=deform_upperarm_fr_base_l, loc=<0 0 0>, rot=<0 0 0 1>}
 deform_upperarm_up_base_l {parent=upperarm_twist_01_l, loc=<-5.9223 -0.7874 1.605>, rot=<0 -0.0007 -0.0191 -0.9998>}
 deform_upperarm_up_l {parent=deform_upperarm_up_base_l, loc=<0 0 0>, rot=<0 0 0 1>}
 upperarm_twist_02_l {parent=upperarm_l, loc=<13.7941 0 0>, rot=<0 0 0 1>}
 clavicle_r {parent=spine_05, loc=<4.3575 -0.1466 2.3223>, rot=<0.7289 0.0465 -0.6829 0.009>}
 upperarm_r {parent=clavicle_r, loc=<-15.5111 0 0>, rot=<0.0226 -0.3234 -0.0088 -0.9459>}
 lowerarm_r {parent=upperarm_r, loc=<-26.6515 0 0>, rot=<0.0391 -0.0099 -0.2453 -0.9686>}
 hand_r {parent=lowerarm_r, loc=<-22.3492 0 0>, rot=<-0.7081 -0.0507 0.0975 -0.6975>}
 index_metacarpal_r {parent=hand_r, loc=<-1.6534 -0.0181 2.9601>, rot=<0.0158 -0.0124 -0.0173 -0.9996>}
 index_01_r {parent=index_metacarpal_r, loc=<-8.3008 -0.0357 0.0524>, rot=<-0.0044 0.0485 0.0924 -0.9945>}
 index_02_r {parent=index_01_r, loc=<-3.3589 0.3003 -0.0453>, rot=<-0.002 0.0027 0.3225 -0.9466>}
 index_03_r {parent=index_02_r, loc=<-2.5578 0.1231 0.0251>, rot=<-0.002 0.0067 0.2558 -0.9667>}
 middle_metacarpal_r {parent=hand_r, loc=<-2.1856 0.1726 0.5984>, rot=<0.006 0.0168 -0.0329 -0.9993>}
 middle_01_r {parent=middle_metacarpal_r, loc=<-7.3273 0.2447 -0.0129>, rot=<-0.0208 0.0042 0.0858 -0.9961>}
 middle_02_r {parent=middle_01_r, loc=<-4.0142 -0.1411 -0.07>, rot=<-0.0217 0.0259 0.3512 -0.9357>}
 middle_03_r {parent=middle_02_r, loc=<-2.6074 -0.0534 -0.1312>, rot=<0.0083 0.0188 0.1708 -0.9851>}
 pinky_metacarpal_r {parent=hand_r, loc=<-1.7617 0.9753 -3.3814>, rot=<-0.0397 0.0814 -0.069 -0.9935>}
 pinky_01_r {parent=pinky_metacarpal_r, loc=<-7.1707 0.4832 0.1094>, rot=<-0.0654 -0.0035 0.2242 -0.9723>}
 pinky_02_r {parent=pinky_01_r, loc=<-2.8274 -0.0892 0.1043>, rot=<0.0037 0.0011 0.2854 -0.9584>}
 pinky_03_r {parent=pinky_02_r, loc=<-2.2045 -0.031 0.0078>, rot=<-0.0002 0.0098 0.1445 -0.9895>}
 ring_metacarpal_r {parent=hand_r, loc=<-2.2073 0.3627 -1.3638>, rot=<-0.0284 0.0722 -0.0458 -0.9959>}
 ring_01_r {parent=ring_metacarpal_r, loc=<-7.0297 0.6197 0.1446>, rot=<-0.0489 -0.0252 0.1892 -0.9804>}
 ring_02_r {parent=ring_01_r, loc=<-3.6171 -0.4648 0.1027>, rot=<0 0 0.2806 -0.9598>}
 ring_03_r {parent=ring_02_r, loc=<-2.4153 -0.1141 -0.0199>, rot=<0 0 0.1189 -0.9929>}
 thumb_01_r {parent=hand_r, loc=<-1.3119 0.3192 2.459>, rot=<0.6546 -0.4042 0.0664 -0.6354>}
 thumb_02_r {parent=thumb_01_r, loc=<-5.9335 0.0447 -0.0401>, rot=<-0.0037 0.0213 0.298 -0.9543>}
 thumb_03_r {parent=thumb_02_r, loc=<-2.2223 -0.3682 0.0907>, rot=<0.0118 -0.013 0.0066 -0.9998>}
 weapon_r {parent=hand_r, loc=<0.1975 3.4112 -0.3811>, rot=<0 0 0 -1>}
 lowerarm_twist_01_r {parent=lowerarm_r, loc=<-17.855 0 0>, rot=<0 0 0 1>}
 lowerarm_twist_02_r {parent=lowerarm_r, loc=<-9.6199 0 0>, rot=<0 0 0 1>}
 upperarm_twist_01_r {parent=upperarm_r, loc=<-0.0041 0 0>, rot=<0 0 0 1>}
 deform_upperarm_bk_base_r {parent=upperarm_twist_01_r, loc=<3.2581 5.546 2.0367>, rot=<0 -0.0007 -0.0191 -0.9998>}
 deform_upperarm_bk_r {parent=deform_upperarm_bk_base_r, loc=<0 0 0>, rot=<0 0 0 1>}
 deform_upperarm_fr_base_r {parent=upperarm_twist_01_r, loc=<2.8462 -3.9432 2.0159>, rot=<0 -0.0007 -0.0191 -0.9998>}
 deform_upperarm_fr_r {parent=deform_upperarm_fr_base_r, loc=<0 0 0>, rot=<0 0 0 1>}
 deform_upperarm_up_base_r {parent=upperarm_twist_01_r, loc=<5.9222 0.7874 -1.6049>, rot=<0 -0.0007 -0.0191 -0.9998>}
 deform_upperarm_up_r {parent=deform_upperarm_up_base_r, loc=<0 0 0>, rot=<0 0 0 1>}
 upperarm_twist_02_r {parent=upperarm_r, loc=<-13.7942 0 0>, rot=<0 0 0 1>}
 neck_01 {parent=spine_05, loc=<11.3138 0.1529 0>, rot=<0 0 -0.2003 -0.9797>}
 neck_02 {parent=neck_01, loc=<4.5627 0 0>, rot=<0 0 0.0504 -0.9987>}
 head {parent=neck_02, loc=<5.3914 0 0>, rot=<0 0 0.115 -0.9934>}
 deform_pec_aim_base_l {parent=spine_04, loc=<6.2615 9.3309 -1.2237>, rot=<0.0603 -0.4033 0.3318 -0.8507>}
 deform_pec_aim_l {parent=deform_pec_aim_base_l, loc=<0 0 0>, rot=<0 0 0 1>}
 deform_pec_base_l {parent=deform_pec_aim_l, loc=<9.3665 0 0>, rot=<0 0 0 1>}
 deform_pec_l {parent=deform_pec_base_l, loc=<0 0 0>, rot=<0 0 0 1>}
 deform_pec_aim_base_r {parent=spine_04, loc=<6.2611 9.3309 1.2237>, rot=<-0.4033 -0.0603 0.8507 0.3318>}
 deform_pec_aim_r {parent=deform_pec_aim_base_r, loc=<0 0 0>, rot=<0 0 0 1>}
 deform_pec_base_r {parent=deform_pec_aim_r, loc=<-9.3667 -0.0003 -0.0004>, rot=<0 0 0 1>}
 deform_pec_r {parent=deform_pec_base_r, loc=<0 0 0>, rot=<0 0 0 1>}
 thigh_l {parent=pelvis, loc=<-2.5693 -0.0109 -7.7501>, rot=<0.0012 0.0572 0.0155 -0.9982>}
 calf_l {parent=thigh_l, loc=<-39.0484 0 0>, rot=<0.0006 -0.0161 -0.0391 -0.9991>}
 calf_twist_01_l {parent=calf_l, loc=<-35.813 -0.0967 0.0397>, rot=<0.0237 0.0095 -0.0035 -0.9997>}
 calf_twist_02_l {parent=calf_l, loc=<-18.8064 -0.0967 0.0397>, rot=<0.0237 0.0095 -0.0035 -0.9997>}
 foot_l {parent=calf_l, loc=<-42.561 0.1077 0.4292>, rot=<0.0418 -0.0455 0.0343 -0.9975>}
 ball_l {parent=foot_l, loc=<-6.1599 13.8264 0.6806>, rot=<0.0001 -0.0001 -0.7063 -0.7079>}
 dyn_knee_L {parent=calf_l, loc=<-9.2527 -6.7873 0>, rot=<0.01 0.0235 -0.7094 -0.7044>}
 deform_knee_in_base_l {parent=calf_l, loc=<3.6268 -5.2446 -0.2814>, rot=<-0.0017 -0.0216 -0.0001 -0.9998>}
 deform_knee_in_l {parent=deform_knee_in_base_l, loc=<0 0 0>, rot=<0 0 0 1>}
 deform_knee_out_base_l {parent=calf_l, loc=<2.8134 3.7791 -0.4614>, rot=<-0.0017 -0.0216 -0.0001 -0.9998>}
 deform_knee_out_l {parent=deform_knee_out_base_l, loc=<0 0 0>, rot=<0 0 0 1>}
 thigh_twist_01_l {parent=thigh_l, loc=<-21.6303 0 0>, rot=<0.0013 0 0 -1>}
 thigh_r {parent=pelvis, loc=<-2.5691 -0.0109 7.7501>, rot=<0.0572 -0.0012 0.9982 0.0155>}
 calf_r {parent=thigh_r, loc=<39.0487 0 0>, rot=<0.0006 -0.0161 -0.0391 -0.9991>}
 calf_twist_01_r {parent=calf_r, loc=<35.8132 0.0967 -0.0397>, rot=<0.0237 0.0095 -0.0035 -0.9997>}
 calf_twist_02_r {parent=calf_r, loc=<18.8064 0.0967 -0.0397>, rot=<0.0237 0.0095 -0.0035 -0.9997>}
 foot_r {parent=calf_r, loc=<42.5612 -0.1077 -0.4292>, rot=<0.0418 -0.0455 0.0343 -0.9975>}
 ball_r {parent=foot_r, loc=<6.1598 -13.8264 -0.6806>, rot=<0.0001 -0.0001 -0.7063 -0.7079>}
 dyn_Knee_R {parent=calf_r, loc=<8.4765 5.7458 0>, rot=<0.01 0.0235 -0.7094 -0.7044>}
 deform_knee_in_base_r {parent=calf_r, loc=<-3.6267 5.2446 0.2813>, rot=<-0.0017 -0.0216 -0.0001 -0.9998>}
 deform_knee_in_r {parent=deform_knee_in_base_r, loc=<0 0 0>, rot=<0 0 0 1>}
 deform_knee_out_base_r {parent=calf_r, loc=<-2.8134 -3.7791 0.4614>, rot=<-0.0017 -0.0216 -0.0001 -0.9998>}
 deform_knee_out_r {parent=deform_knee_out_base_r, loc=<0 0 0>, rot=<0 0 0 1>}
 thigh_twist_01_r {parent=thigh_r, loc=<21.6303 0 0>, rot=<0.0013 0 0 -1>}
 dyn_coat_MSTR_frnt_r {parent=pelvis, loc=<11.3406 7.5127 7.226>, rot=<0.1203 -0.2308 0.9616 0.0866>}
 dyn_coat_frnt_r {parent=dyn_coat_MSTR_frnt_r, loc=<0.0674 0 0>, rot=<0 0 0 1>}
 dyn_coat_frnt_01_r {parent=dyn_coat_frnt_r, loc=<9.9128 0 0>, rot=<0 0 0 1>}
 dyn_coat_MSTR_frnt_l {parent=pelvis, loc=<6.0124 8.8592 -5.8759>, rot=<-0.1203 0.2308 0.9616 0.0866>}
 dyn_coat_frnt_l {parent=dyn_coat_MSTR_frnt_l, loc=<0.0674 0 0>, rot=<0 0 0 1>}
 dyn_coat_frnt_01_l {parent=dyn_coat_frnt_l, loc=<9.9128 0 0>, rot=<0 0 0 1>}
 dyn_coat_MSTR_back_l {parent=pelvis, loc=<5.615 -10.6358 -7.0829>, rot=<0.0621 -0.9636 -0.2499 -0.0719>}
 dyn_coat_back_l {parent=dyn_coat_MSTR_back_l, loc=<0.0674 0 0>, rot=<0 0 0 1>}
 dyn_coat_back_01_l {parent=dyn_coat_back_l, loc=<9.3825 0 0>, rot=<0 0 0 1>}
 dyn_coat_MSTR_back_r {parent=pelvis, loc=<5.615 -10.6358 6.9916>, rot=<0.0623 -0.9649 0.2446 0.0723>}
 dyn_coat_back_r {parent=dyn_coat_MSTR_back_r, loc=<0.0674 0 0>, rot=<0 0 0 1>}
 dyn_coat_back_01_r {parent=dyn_coat_back_r, loc=<9.3825 0 0>, rot=<0 0 0 1>}
 dyn_coat_MSTR_side_r {parent=pelvis, loc=<8.074 -0.241 14.225>, rot=<0.0622 -0.7046 0.7019 0.0839>}
 dyn_coat_side_r {parent=dyn_coat_MSTR_side_r, loc=<0.0674 0 0>, rot=<0 0 0 1>}
 dyn_coat_side_01_r {parent=dyn_coat_side_r, loc=<8.3923 0 0>, rot=<0 0 0 1>}
 dyn_coat_MSTR_side_l {parent=pelvis, loc=<8.074 -0.241 -14.2245>, rot=<0.0651 -0.7039 -0.702 -0.0869>}
 dyn_coat_side_l {parent=dyn_coat_MSTR_side_l, loc=<0.0674 0 0>, rot=<0 0 0 1>}
 dyn_coat_side_01_l {parent=dyn_coat_side_l, loc=<8.3923 0 0>, rot=<0 0 0 1>}
 dyn_glute {parent=pelvis, loc=<-10.1133 -4.6149 -5.5076>, rot=<0.0513 -0.9915 -0.097 -0.0699>}
 dyn_MSTR_side_frt_R {parent=pelvis, loc=<7.9348 4.374 13.1528>, rot=<-0.0577 -0.3028 0.9404 0.1438>}
 dyn_side_frt_R {parent=dyn_MSTR_side_frt_R, loc=<0.0674 0 0>, rot=<0 0 0 1>}
 dyn_side_frt_R_01 {parent=dyn_side_frt_R, loc=<8.3923 0 0>, rot=<0 0 0 1>}
 dyn_MSTR_frnt_side_l {parent=pelvis, loc=<8.0411 4.039 -12.1309>, rot=<0.0366 0.355 0.9202 0.1607>}
 dyn_frn_sidet_l {parent=dyn_MSTR_frnt_side_l, loc=<0.0674 0 0>, rot=<0 0 0 1>}
 dyn_frnt_side_01_l {parent=dyn_frn_sidet_l, loc=<9.9128 0 0>, rot=<0 0 0 1>}
 dyn_pelvis_scale {parent=pelvis, loc=<0 0 0>, rot=<0 0 0 1>, scale=<0.9353 0.9353 0.9353>}
]
