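# Post-process style AnimBlueprint for the M_MED_Henchman_Bad player body, adding simulated coat motion to the pose copied from the parent mesh.
# Three Control Rig and RigidBody branches, each with its own physics asset and some coat bone offsets, are selected by crouch and skydive bools.
# The blended pose gets a pelvis helper bone scale, then a deformation Control Rig.
# The update event stores a deprecated state-based gravity override that the AnimGraph never reads.
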AnimBlueprint /Game/Characters/Player/Male/Medium/Bodies/M_MED_Henchman_Bad/Meshes/M_MED_Henchman_Bad_AnimBP
target_skeleton=&/M_MED_Henchman_Bad_Skeleton
graphs [
 AnimGraph {nodes [
   n1 = "Transform (Modify) Bone - Bone: dyn_coat_MSTR_back_l"(Alpha=1, Translation="0.349,-1.092,-1.120", Rotation="0,20,0", Scale="1,1,1") {
    rotationMode=BMM_Additive
    rotationSpace=BCS_BoneSpace
    translationMode=BMM_Additive
    translationSpace=BCS_BoneSpace
    alphaBoolBlend {blendOption=Linear}
    boneToModify {boneName=dyn_coat_MSTR_back_l}
    rotation {yaw=20}
    translation=<0.349 -1.092 -1.12>
   }
   "Transform (Modify) Bone - Bone: dyn_coat_MSTR_back_r"(ComponentPose=@n1, Alpha=1, Translation="0.251,-2.012,0.486", Rotation="0,10,0", Scale="1,1,1") {
    rotationMode=BMM_Additive
    rotationSpace=BCS_BoneSpace
    translationMode=BMM_Additive
    translationSpace=BCS_BoneSpace
    alphaBoolBlend {blendOption=Linear}
    boneToModify {boneName=dyn_coat_MSTR_back_r}
    rotation {yaw=10}
    translation=<0.251 -2.012 0.486>
   }
   n2 = Control Rig(Source=Copy Pose From Mesh() {bUseAttachedParent=true}) {
    alphaBoolBlend {blendOption=Linear}
    controlRigAssetReference {blueprintRigClass=&/M_MED_Henchman_Bad_crouch_CtrlRig.M_MED_Henchman_Bad_crouch_CtrlRig_C}
   }
   n3 = "Transform (Modify) Bone - Bone: dyn_coat_MSTR_back_l"(ComponentPose=Local To Component(LocalPose=@n2), Alpha=1, Translation="0.258,-3.253,0.028", Rotation="0,20,0", Scale="1,1,1") {
    rotationMode=BMM_Additive
    rotationSpace=BCS_BoneSpace
    translationMode=BMM_Additive
    translationSpace=BCS_BoneSpace
    alphaBoolBlend {blendOption=Linear}
    boneToModify {boneName=dyn_coat_MSTR_back_l}
    rotation {yaw=20}
    translation=<0.258 -3.253 0.028>
   }
   n4 = RigidBody!(ComponentPose=@n3, Alpha, OverrideWorldGravity:Vector="0,0,-500", ExternalForce) -> (Pose) {
    bOverrideWorldGravity=true
    overridePhysicsAsset=&/M_MED_Henchman_Bad2_CrouchPhysics
    simulationSpace=BaseBoneSpace
    alphaBoolBlend {blendOption=Linear}
    baseBoneRef {boneName=spine_01}
    componentLinearAccScale {z=0.1}
    componentLinearVelScale {z=0.1}
    overrideWorldGravity {z=-980}
    simSpaceSettings {
     maxAngularAcceleration=200
     maxAngularVelocity=3
     maxLinearAcceleration=600
     maxLinearVelocity=200
     velocityScaleZ=0
     worldAlpha=1
     externalLinearDragV=<1 1 1>
    }
   }
   n5 = Control Rig(Source=Copy Pose From Mesh() {bUseAttachedParent=true}) {
    alphaBoolBlend {blendOption=Linear}
    controlRigAssetReference {blueprintRigClass=&/M_MED_Henchman_Bad_CtrlRig.M_MED_Henchman_Bad_CtrlRig_C}
   }
   n6 = RigidBody(ComponentPose=Local To Component(LocalPose=@n5), OverrideWorldGravity="0,0,-500") {
    bOverrideWorldGravity=true
    overridePhysicsAsset=&/M_MED_Henchman_Bad2_Physics_sky_drive
    simulationSpace=BaseBoneSpace
    alphaBoolBlend {blendOption=Linear}
    baseBoneRef {boneName=pelvis}
    overrideWorldGravity {z=-980}
    simSpaceSettings {maxAngularAcceleration=100, maxAngularVelocity=3, maxLinearAcceleration=600, maxLinearVelocity=160, velocityScaleZ=0, worldAlpha=1}
   }
   n7 = Blend Poses by bool(BlendPose_0=Component To Local(ComponentPose=@n4), BlendPose_1=Component To Local(ComponentPose=@n6), bActiveValue=Get bIsCrouching()) {blendType=Linear}
   n8 = Control Rig(Source=Copy Pose From Mesh() {bUseAttachedParent=true}) {
    alphaBoolBlend {blendOption=Linear}
    controlRigAssetReference {blueprintRigClass=&M_MED_Henchman_Bad_CtrlRig_C}
   }
   n9 = RigidBody(ComponentPose=Local To Component(LocalPose=@n8), OverrideWorldGravity="0,0,-500") {
    bOverrideWorldGravity=true
    overridePhysicsAsset=&/M_MED_Henchman_Bad2_Physics
    simulationSpace=BaseBoneSpace
    alphaBoolBlend {blendOption=Linear}
    baseBoneRef {boneName=spine_01}
    componentLinearAccScale {z=0.1}
    componentLinearVelScale {z=0.1}
    overrideWorldGravity {z=-980}
    simSpaceSettings {
     maxAngularAcceleration=50
     maxAngularVelocity=3
     maxLinearAcceleration=600
     maxLinearVelocity=160
     velocityScaleZ=0
     worldAlpha=1
     externalLinearDragV=<1 1 1>
    }
   }
   n10 = "Transform (Modify) Bone - Bone: dyn_coat_MSTR_frnt_l"(ComponentPose=@n9, Alpha=1, Translation="0,0,0", Rotation="-0.651,-28.734,-32.643", Scale="1,1,1") {
    rotationMode=BMM_Additive
    rotationSpace=BCS_BoneSpace
    translationMode=BMM_Additive
    translationSpace=BCS_BoneSpace
    alphaBoolBlend {blendOption=Linear}
    boneToModify {boneName=dyn_coat_MSTR_frnt_l}
    rotation=<rot -0.651 -28.734 -32.643>
   }
   n11 = Component To Local(ComponentPose=@n10)
   n12 = Get bIsSkyDiving()
   n13 = "Transform (Modify) Bone - Bone: dyn_pelvis_scale"(
    ComponentPose=Local To Component(LocalPose=Blend Poses by bool(BlendPose_0=@n11, BlendPose_1=@n7, bActiveValue=@n12) {blendType=Linear})
    Alpha=1
    Translation="0,0,0"
    Rotation="0,0,0"
    Scale="0.5,0.5,0.5"
   ) {
    rotationMode=BMM_Additive
    scaleMode=BMM_Replace
    translationMode=BMM_Additive
    alphaBoolBlend {blendInTime=0.5, blendOption=Linear, blendOutTime=1}
    alphaScaleBiasClamp {bInterpResult=true, interpSpeedDecreasing=15, interpSpeedIncreasing=15}
    boneToModify {boneName=dyn_pelvis_scale}
    scale=<0.5 0.5 0.5>
   }
   n14 = Control Rig(Source=Component To Local(ComponentPose=@n13)) {
    alphaBoolBlend {blendOption=Linear}
    controlRigAssetReference {blueprintRigClass=&/F_MED_Henchman_Bad_Deforms_ControlRig.F_MED_Henchman_Bad_Deforms_ControlRig_C}
   }
   Output Pose(Result=@n14) {layerGroup=None}
  ]}
 EventGraph {nodes [
   Event BlueprintUpdateAnimation()
   > n1 = Gravity Override State Function DEPRECATED(CharacterPartAnimInstance=Self-Reference(), GameTimeInSeconds=GetGameTimeInSeconds(), GravityOverrideParams=Get Gravity Override Params())
   > Set Gravity Override_Loc(Gravity Override_Loc:Vector=@n1) -> (Output_Get:Vector)
   > Set craunch_bp(craunch_bp:bool=Get bIsCrouching()) -> (Output_Get:bool)
  ]}
]
variables [
 craunch_bp {type=boolean}
 Gravity Override Params {
  type=GravityOverrideParamsStruct
  default {
   skydive_DivingDown {gravityOverride=<-150 0 0>, windAmplitude=<50 33 111>, windFrequency=<1.5 12 3>}
   skydive_FloatingDown {gravityOverride=<0 -700 333>, windAmplitude=<-20 -500 111>, windFrequency=<1.5 7 3>, windOffset=<0 -300 -111>}
   skydive_GoingUp {gravityOverride=<0 -300 333>, windAmplitude=<200 -500 111>, windFrequency=<1.5 7 3>}
   skydive_Parachuting {gravityOverride=<0 -800 -30>, windAmplitude=<200 -500 111>, windFrequency=<1.5 7 3>}
  }
 }
 Gravity Override Params_1 {
  type=GravityOverrideParamsStruct
  default {
   dBNO {gravityOverride=<0 0 0>}
   dBNO_Carried {gravityOverride=<-980 0 0>}
   emoteOrMelee {gravityOverride=<0 0 0>}
   front_End {gravityOverride=<0 0 0>}
   onGround_Falling {gravityOverride=<0 0 0>}
   onGround_Moving {gravityOverride=<0 0 0>}
   onGround_Standing {gravityOverride=<0 0 0>}
   rail_Grinding {gravityOverride=<0 0 0>}
   riding_Moving {gravityOverride=<0 0 0>}
   riding_Moving_Targeting {gravityOverride=<0 0 0>}
   riding_Standing {gravityOverride=<0 0 0>}
   skydive_DivingDown {gravityOverride=<0 0 0>, windAmplitude=<0 0 0>, windFrequency=<0 0 0>}
   skydive_FloatingDown {gravityOverride=<0 0 0>, windAmplitude=<1 0 0>, windFrequency=<0 0 0>, windOffset=<0 0 0>}
   skydive_GoingUp {gravityOverride=<0 0 0>, windAmplitude=<0 0 0>, windFrequency=<0 0 0>}
   skydive_Parachuting {gravityOverride=<0 0 0>, windAmplitude=<0 0 0>, windFrequency=<0 0 0>}
   skydive_Surfing {gravityOverride=<0 0 0>, windAmplitude=<0 0 0>, windFrequency=<1.5 0 0>}
   swimming {gravityOverride=<0 0 0>}
   tactical_Sprint {gravityOverride=<0 0 0>}
   ziplining {gravityOverride=<0 0 0>}
  }
 }
 Gravity Override_Loc {type=Vector}
]
